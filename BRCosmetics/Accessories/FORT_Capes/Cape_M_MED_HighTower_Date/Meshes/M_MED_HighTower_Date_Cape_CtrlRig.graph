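# Control Rig for the HighTower Date cape, procedurally posing its chain and shoulder helper bones from body motion.
# Weighted transform constraints place both chain bases from neck, spine and clavicles, blend inner links toward the chain end, then aim each link at it.
# Right upper-arm angle drives two staged additive shoulder offsets, the second faded by isSkydiving.
# Left arm angle drives only dyn_shoulderScale_l, which is absent from the listed hierarchy.
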ControlRigBlueprint /BRCosmetics/Accessories/FORT_Capes/Cape_M_MED_HighTower_Date/Meshes/M_MED_HighTower_Date_Cape_CtrlRig
graphs [
 RigVMFunctionLibrary {links [], nodes []}
 RigVMModel {nodes [
   n1 = &RigUnit_BeginExecution()
   /Script/ControlRig.RigUnit_TransformConstraint(
    BaseTransform=(Rotation=(X=0,Y=0,Z=0,W=1),Translation=(X=0,Y=0,Z=0),Scale3D=(X=1,Y=1,Z=1))
    BaseTransformSpace=GlobalSpace
    Bone=dyn_capeChain_01_Base_r
    Targets="((Transform=(Rotation=(X=0,Y=0,Z=0,W=1),Translation=(X=0,Y=0,Z=0),Scale3D=(X=1,Y=1,Z=1)),Weight=0.0716,bMaintainOffset=True,Filter=(TranslationFilter=(bX=True,bY=True,bZ=True),RotationFilter=(bX=True,bY=True,bZ=True),ScaleFilter=(bX=True,bY=True,bZ=True))) ...(truncated)"
    bUseInitialTransforms=True
    Targets.0.Transform=&RigUnit_GetBoneTransform(Bone=clavicle_cape_r).Transform
    Targets.1.Transform=&RigUnit_GetBoneTransform(Bone=neck_02).Transform
    Targets.2.Transform=&RigUnit_GetBoneTransform(Bone=spine_03).Transform
    Targets.3.Transform=&RigUnit_GetBoneTransform(Bone=spine_04).Transform
    Targets.4.Transform=&RigUnit_GetBoneTransform(Bone=spine_05).Transform
    Targets.5.Transform=&RigUnit_GetBoneTransform(Bone=clavicle_r).Transform
    ExecutePin=@n1.ExecutePin
   ) {is=struct RigVMUnitNode}
   > &RigUnit_TransformConstraint(
    BaseTransform=(Rotation=(X=0,Y=0,Z=0,W=1),Translation=(X=0,Y=0,Z=0),Scale3D=(X=1,Y=1,Z=1))
    BaseTransformSpace=GlobalSpace
    Bone=dyn_capeChain_01_Base_l
    Targets="((Transform=(Rotation=(X=0,Y=0,Z=0,W=1),Translation=(X=0,Y=0,Z=0),Scale3D=(X=1,Y=1,Z=1)),Weight=0.0716,bMaintainOffset=True,Filter=(TranslationFilter=(bX=True,bY=True,bZ=True),RotationFilter=(bX=True,bY=True,bZ=True),ScaleFilter=(bX=True,bY=True,bZ=True))) ...(truncated)"
    bUseInitialTransforms=True
    Targets.1.Transform=&RigUnit_GetBoneTransform(Bone=neck_02).Transform
    Targets.2.Transform=&RigUnit_GetBoneTransform(Bone=spine_03).Transform
    Targets.3.Transform=&RigUnit_GetBoneTransform(Bone=spine_04).Transform
    Targets.5.Transform=&RigUnit_GetBoneTransform(Bone=clavicle_l).Transform
    Targets.4.Transform=&RigUnit_GetBoneTransform(Bone=spine_05).Transform
    Targets.0.Transform=&RigUnit_GetBoneTransform(Bone=clavicle_cape_l).Transform
   )
   n2 = &RigUnit_GetBoneTransform(Bone=dyn_capeChainEnd)
   n3 = &RigUnit_GetBoneTransform(Bone=dyn_capeChain_01_Base_r)
   > &RigUnit_TransformConstraint(
    BaseTransform=(Rotation=(X=0,Y=0,Z=0,W=1),Translation=(X=0,Y=0,Z=0),Scale3D=(X=1,Y=1,Z=1))
    BaseTransformSpace=GlobalSpace
    Bone=dyn_capeChain_02_r
    Targets="((Transform=(Rotation=(X=0,Y=0,Z=0,W=1),Translation=(X=0,Y=0,Z=0),Scale3D=(X=1,Y=1,Z=1)),Weight=0.4,bMaintainOffset=True,Filter=(TranslationFilter=(bX=True,bY=True,bZ=True),RotationFilter=(bX=True,bY=True,bZ=True),ScaleFilter=(bX=True,bY=True,bZ=True))),(T ...(truncated)"
    bUseInitialTransforms=True
    Targets.0.Transform=@n2.Transform
    Targets.1.Transform=@n3.Transform
   )
   > &RigUnit_TransformConstraint(
    BaseTransform=(Rotation=(X=0,Y=0,Z=0,W=1),Translation=(X=0,Y=0,Z=0),Scale3D=(X=1,Y=1,Z=1))
    BaseTransformSpace=GlobalSpace
    Bone=dyn_capeChain_03_r
    Targets="((Transform=(Rotation=(X=0,Y=0,Z=0,W=1),Translation=(X=0,Y=0,Z=0),Scale3D=(X=1,Y=1,Z=1)),Weight=0.7,bMaintainOffset=True,Filter=(TranslationFilter=(bX=True,bY=True,bZ=True),RotationFilter=(bX=True,bY=True,bZ=True),ScaleFilter=(bX=True,bY=True,bZ=True))),(T ...(truncated)"
    bUseInitialTransforms=True
    Targets.0.Transform=@n2.Transform
    Targets.1.Transform=@n3.Transform
   )
   n4 = &RigUnit_GetBoneTransform(Bone=dyn_capeChain_01_Base_l)
   > &RigUnit_TransformConstraint(
    BaseTransform=(Rotation=(X=0,Y=0,Z=0,W=1),Translation=(X=0,Y=0,Z=0),Scale3D=(X=1,Y=1,Z=1))
    BaseTransformSpace=GlobalSpace
    Bone=dyn_capeChain_02_l
    Targets="((Transform=(Rotation=(X=0,Y=0,Z=0,W=1),Translation=(X=0,Y=0,Z=0),Scale3D=(X=1,Y=1,Z=1)),Weight=0.4,bMaintainOffset=True,Filter=(TranslationFilter=(bX=True,bY=True,bZ=True),RotationFilter=(bX=True,bY=True,bZ=True),ScaleFilter=(bX=True,bY=True,bZ=True))),(T ...(truncated)"
    bUseInitialTransforms=True
    Targets.0.Transform=@n2.Transform
    Targets.1.Transform=@n4.Transform
   )
   > &RigUnit_TransformConstraint(
    BaseTransform=(Rotation=(X=0,Y=0,Z=0,W=1),Translation=(X=0,Y=0,Z=0),Scale3D=(X=1,Y=1,Z=1))
    BaseTransformSpace=GlobalSpace
    Bone=dyn_capeChain_03_l
    Targets="((Transform=(Rotation=(X=0,Y=0,Z=0,W=1),Translation=(X=0,Y=0,Z=0),Scale3D=(X=1,Y=1,Z=1)),Weight=0.7,bMaintainOffset=True,Filter=(TranslationFilter=(bX=True,bY=True,bZ=True),RotationFilter=(bX=True,bY=True,bZ=True),ScaleFilter=(bX=True,bY=True,bZ=True))),(T ...(truncated)"
    bUseInitialTransforms=True
    Targets.0.Transform=@n2.Transform
    Targets.1.Transform=@n4.Transform
   )
   n5 = &RigUnit_GetBoneTransform(Bone=dyn_capeChainEnd)
   > /Script/ControlRig.RigUnit_AimBone(
    Bone=dyn_capeChain_01_r
    DebugSettings=(Scale=10,WorldOffset=(Rotation=(X=0,Y=0,Z=0,W=1),Translation=(X=0,Y=0,Z=0),Scale3D=(X=1,Y=1,Z=1)))
    Primary=(Weight=1,Axis=(X=1,Y=0,Z=0),Target=(X=1,Y=0,Z=0),Kind=Location)
    Secondary=(Weight=1,Axis=(X=0,Y=0,Z=0),Target=(X=1,Y=0,Z=0),Kind=Location)
    Weight=1
    bPropagateToChildren=True
    Primary.Target=@n5.Transform.Translation
   ) {is=struct RigVMUnitNode}
   > &RigUnit_AimBone(
    Bone=dyn_capeChain_02_r
    DebugSettings=(Scale=10,WorldOffset=(Rotation=(X=0,Y=0,Z=0,W=1),Translation=(X=0,Y=0,Z=0),Scale3D=(X=1,Y=1,Z=1)))
    Primary=(Weight=1,Axis=(X=1,Y=0,Z=0),Target=(X=1,Y=0,Z=0),Kind=Location)
    Secondary=(Weight=1,Axis=(X=0,Y=0,Z=0),Target=(X=1,Y=0,Z=0),Kind=Location)
    Weight=1
    bPropagateToChildren=True
    Primary.Target=@n5.Transform.Translation
   )
   > &RigUnit_AimBone(
    Bone=dyn_capeChain_03_r
    DebugSettings=(Scale=10,WorldOffset=(Rotation=(X=0,Y=0,Z=0,W=1),Translation=(X=0,Y=0,Z=0),Scale3D=(X=1,Y=1,Z=1)))
    Primary=(Weight=1,Axis=(X=1,Y=0,Z=0),Target=(X=1,Y=0,Z=0),Kind=Location)
    Secondary=(Weight=1,Axis=(X=0,Y=0,Z=0),Target=(X=1,Y=0,Z=0),Kind=Location)
    Weight=1
    bPropagateToChildren=True
    Primary.Target=@n5.Transform.Translation
   )
   > &RigUnit_AimBone(
    Bone=dyn_capeChain_01_l
    DebugSettings=(Scale=10,WorldOffset=(Rotation=(X=0,Y=0,Z=0,W=1),Translation=(X=0,Y=0,Z=0),Scale3D=(X=1,Y=1,Z=1)))
    Primary=(Weight=1,Axis=(X=1,Y=0,Z=0),Target=(X=1,Y=0,Z=0),Kind=Location)
    Secondary=(Weight=1,Axis=(X=0,Y=0,Z=0),Target=(X=1,Y=0,Z=0),Kind=Location)
    Weight=1
    bPropagateToChildren=True
    Primary.Target=@n5.Transform.Translation
   )
   > &RigUnit_AimBone(
    Bone=dyn_capeChain_02_l
    DebugSettings=(Scale=10,WorldOffset=(Rotation=(X=0,Y=0,Z=0,W=1),Translation=(X=0,Y=0,Z=0),Scale3D=(X=1,Y=1,Z=1)))
    Primary=(Weight=1,Axis=(X=1,Y=0,Z=0),Target=(X=1,Y=0,Z=0),Kind=Location)
    Secondary=(Weight=1,Axis=(X=0,Y=0,Z=0),Target=(X=1,Y=0,Z=0),Kind=Location)
    Weight=1
    bPropagateToChildren=True
    Primary.Target=@n5.Transform.Translation
   )
   > &RigUnit_AimBone(
    Bone=dyn_capeChain_03_l
    DebugSettings=(Scale=10,WorldOffset=(Rotation=(X=0,Y=0,Z=0,W=1),Translation=(X=0,Y=0,Z=0),Scale3D=(X=1,Y=1,Z=1)))
    Primary=(Weight=1,Axis=(X=1,Y=0,Z=0),Target=(X=1,Y=0,Z=0),Kind=Location)
    Secondary=(Weight=1,Axis=(X=0,Y=0,Z=0),Target=(X=1,Y=0,Z=0),Kind=Location)
    Weight=1
    bPropagateToChildren=True
    Primary.Target=@n5.Transform.Translation
   )
   n6 = /Script/RigVM.RigVMFunction_MathFloatAbs(Value=&RigVMFunction_MathQuaternionToEuler(RotationOrder=ZYX, Value=&RigUnit_GetRelativeBoneTransform(Bone=upperarm_r, Space=spine_05).Transform.Rotation).Result.Z) {is=struct RigVMUnitNode, template="Absolute::Execute(in Value,out Result)"}
   > /Script/ControlRig.RigUnit_ModifyBoneTransforms(
    BoneToModify="((Bone=\"dyn_capeFrontShoulder_r\",Transform=(Rotation=(X=0,Y=0,Z=0,W=1),Translation=(X=-1.9,Y=-3.5,Z=-1.25),Scale3D=(X=1,Y=1,Z=1))),(Bone=\"dyn_capeTopShoulder_r\",Transform=(Rotation=(X=0,Y=0,Z=0,W=1),Translation=(X=0,Y=0,Z=0),Scale3D=(X=1,Y=1,Z=1))),(Bone=\" ...(truncated)"
    Mode=AdditiveLocal
    WeightMaximum=1
    Weight=&RigVMFunction_MathFloatMul(
     B=1
     A=&RigVMFunction_MathFloatRemap(SourceMaximum=123, SourceMinimum=30, TargetMinimum=0.15, bClamp=true, Value=@n6.Result).Result
    ) {template="Multiply::Execute(in A,in B,out Result)"}.Result
    BoneToModify.0.Transform.Rotation=&RigVMFunction_MathQuaternionFromEuler(Euler=(X=0,Y=0,Z=0)).Result
    BoneToModify.1.Transform.Rotation=&RigVMFunction_MathQuaternionFromEuler(Euler=(X=0,Y=0,Z=-45)).Result
    BoneToModify.2.Transform.Rotation=&RigVMFunction_MathQuaternionFromEuler(Euler=(X=0,Y=0,Z=-45)).Result
   ) {is=struct RigVMUnitNode}
   n7 = RigVMVariableNode() {variable=isSkydiving}
   > &RigUnit_ModifyBoneTransforms(
    BoneToModify="((Bone=\"dyn_capeBackShoulder_Base_r\",Transform=(Rotation=(X=0,Y=0,Z=0,W=1),Translation=(X=0,Y=-5,Z=15),Scale3D=(X=1,Y=1,Z=1))),(Bone=\"dyn_capeTopShoulder_Base_r\",Transform=(Rotation=(X=0,Y=0,Z=0,W=1),Translation=(X=0,Y=0,Z=6),Scale3D=(X=1,Y=1,Z=1))),(Bone= ...(truncated)"
    Mode=AdditiveLocal
    WeightMaximum=1
    Weight=&RigVMFunction_MathFloatMul(
     A=&RigVMFunction_MathFloatRemap(SourceMaximum=141, SourceMinimum=123, TargetMinimum=0, bClamp=true, Value=@n6.Result).Result
     B=/Script/RigVM.RigVMFunction_MathFloatSub(A=1, B=@n7.Value) {is=struct RigVMUnitNode, template="Subtract::Execute(in A,in B,out Result)"}.Result
    ) {template="Multiply::Execute(in A,in B,out Result)"}.Result
    BoneToModify.0.Transform.Rotation=&RigVMFunction_MathQuaternionFromEuler(Euler=(X=-20,Y=20,Z=-10)).Result
    BoneToModify.1.Transform.Rotation=&RigVMFunction_MathQuaternionFromEuler(Euler=(X=0,Y=0,Z=-30)).Result
    BoneToModify.2.Transform.Rotation=&RigVMFunction_MathQuaternionFromEuler(Euler=(X=0,Y=0,Z=-25)).Result
   )
   n8 = &RigVMFunction_MathQuaternionToEuler(RotationOrder=ZYX, Value=&RigUnit_GetRelativeBoneTransform(Bone=upperarm_l, Space=clavicle_l).Transform.Rotation)
   > &RigUnit_ModifyBoneTransforms(
    BoneToModify=((Bone="dyn_shoulderScale_l",Transform=(Rotation=(X=0,Y=0,Z=0,W=1),Translation=(X=0,Y=0,Z=0),Scale3D=(X=1,Y=1,Z=1))))
    Mode=AdditiveLocal
    WeightMaximum=1
    Weight=&RigVMFunction_MathFloatRemap(
     SourceMaximum=27
     SourceMinimum=74
     TargetMinimum=0
     bClamp=true
     Value=&RigVMFunction_MathFloatAbs(Value=@n8.Result.Z) {template="Absolute::Execute(in Value,out Result)"}.Result
    ).Result
    BoneToModify.0.Transform.Rotation=&RigVMFunction_MathQuaternionFromEuler(Euler=(X=-50,Y=10,Z=-10)).Result
   )
   RigUnit_BeginExecution = &RigUnit_BeginExecution()
  ]}
]
hierarchy [
 root {loc=<0 0 0>, rot=<0 0 0 1>}
 pelvis {parent=root, loc=<0 -0.7071 93.1191>, rot=<0.0127 0.707 -0.0127 -0.707>}
 spine_01 {parent=pelvis, loc=<6.7494 -0.6388 0>, rot=<0 0 0.034 -0.9994>}
 spine_02 {parent=spine_01, loc=<7.2518 -0.035 0>, rot=<0 0 0.0238 -0.9997>}
 spine_03 {parent=spine_02, loc=<6.0104 0 0>, rot=<0 0 0.0149 -0.9999>}
 spine_04 {parent=spine_03, loc=<8.1024 0 0>, rot=<0 0 -0.0132 -0.9999>}
 spine_05 {parent=spine_04, loc=<10.0146 0 0>, rot=<0 0 -0.0015 -1>}
 clavicle_l {parent=spine_05, loc=<4.3575 -0.1466 -2.3223>, rot=<0.0465 -0.7289 -0.009 -0.6829>}
 upperarm_l {parent=clavicle_l, loc=<15.5111 0 0>, rot=<0.0226 -0.3234 -0.0088 -0.9459>}
 lowerarm_l {parent=upperarm_l, loc=<26.6514 0 0>, rot=<0.0391 -0.0099 -0.2453 -0.9686>}
 clavicle_cape_l {parent=clavicle_l, loc=<0 0.0001 0.0003>, rot=<0 0 0 -1>}
 cape_01_l {parent=clavicle_cape_l, loc=<13.08 -4.0816 7.516>, rot=<0.3413 -0.5667 0.2993 -0.6876>}
 cape_02_l {parent=cape_01_l, loc=<6.892 0 0>, rot=<-0.0025 0.0084 -0.2906 -0.9568>}
 cape_03_l {parent=cape_02_l, loc=<11.8886 0 0>, rot=<0 0 -0.1305 -0.9914>}
 cape_04_l {parent=cape_03_l, loc=<23.3505 0 0>, rot=<0 0 -0.0306 -0.9995>}
 cape_05_l {parent=cape_04_l, loc=<20.0893 0 0>, rot=<0 0 -0.008 -1>}
 cape_06_l {parent=cape_05_l, loc=<18.5056 0 0>, rot=<0 0 -0.0041 -1>}
 cape_07_l {parent=cape_06_l, loc=<13.5867 0 0>, rot=<0 0 0.0015 -1>}
 cape_08_l {parent=cape_07_l, loc=<15.1001 0 0>, rot=<0 0 -0.0027 -1>}
 dyn_capeBackShoulder_Base_l {parent=clavicle_cape_l, loc=<8.0681 -5.9931 9.6931>, rot=<0.5954 -0.1376 -0.0578 -0.7894>}
 dyn_capeBackShoulder_l {parent=dyn_capeBackShoulder_Base_l, loc=<0 0 0>, rot=<0 0 0 -1>}
 dyn_capeTopShoulder_Base_l {parent=clavicle_cape_l, loc=<8.4872 -3.1592 9.4987>, rot=<0.7388 -0.026 -0.1127 -0.6639>}
 dyn_capeTopShoulder_l {parent=dyn_capeTopShoulder_Base_l, loc=<0 0 0>, rot=<0 0 0 1>}
 dyn_capeFrontShoulder_Base_l {parent=clavicle_cape_l, loc=<6.0819 5.2175 6.6089>, rot=<0.8836 0.0378 -0.1728 -0.4335>}
 dyn_capeFrontShoulder_l {parent=dyn_capeFrontShoulder_Base_l, loc=<0 0 0>, rot=<0 0 0 1>}
 clavicle_r {parent=spine_05, loc=<4.3575 -0.1466 2.3223>, rot=<0.7289 0.0465 -0.6829 0.009>}
 upperarm_r {parent=clavicle_r, loc=<-15.5111 0 0>, rot=<0.0226 -0.3234 -0.0088 -0.9459>}
 lowerarm_r {parent=upperarm_r, loc=<-26.6514 0 0>, rot=<0.0391 -0.0099 -0.2453 -0.9686>}
 clavicle_cape_r {parent=clavicle_r, loc=<0 0 0>, rot=<0 0 0 -1>}
 cape_01_r {parent=clavicle_cape_r, loc=<-13.0799 4.0816 -7.5166>, rot=<0.5667 0.3413 0.6876 0.2993>}
 cape_02_r {parent=cape_01_r, loc=<6.8922 -0.0002 0>, rot=<0.0025 -0.0084 -0.2906 -0.9568>}
 cape_03_r {parent=cape_02_r, loc=<11.8886 0 0>, rot=<0 0 -0.1305 -0.9914>}
 cape_04_r {parent=cape_03_r, loc=<23.3501 0.0001 0>, rot=<0 0 -0.0306 -0.9995>}
 cape_05_r {parent=cape_04_r, loc=<20.0898 0 0.0001>, rot=<0 0 -0.008 -1>}
 cape_06_r {parent=cape_05_r, loc=<18.5056 -0.0001 -0.0001>, rot=<0 0 -0.0041 -1>}
 cape_07_r {parent=cape_06_r, loc=<13.5866 0 0.0001>, rot=<0 0 0.0015 -1>}
 cape_08_r {parent=cape_07_r, loc=<15.1001 0.0001 -0.0001>, rot=<0 0 -0.0027 -1>}
 dyn_capeFrontShoulder_Base_r {parent=clavicle_cape_r, loc=<-6.0818 -5.2176 -6.6095>, rot=<0.0172 -0.8265 -0.5345 0.176>}
 dyn_capeFrontShoulder_r {parent=dyn_capeFrontShoulder_Base_r, loc=<0 0 0>, rot=<0 0 0 -1>}
 dyn_capeTopShoulder_Base_r {parent=clavicle_cape_r, loc=<-8.4872 3.1591 -9.4985>, rot=<-0.0269 -0.733 -0.6703 0.1124>}
 dyn_capeTopShoulder_r {parent=dyn_capeTopShoulder_Base_r, loc=<0 0 0>, rot=<0 0 0 -1>}
 dyn_capeBackShoulder_Base_r {parent=clavicle_cape_r, loc=<-8.0681 5.9931 -9.6935>, rot=<0.1335 0.6463 0.7484 -0.0668>}
 dyn_capeBackShoulder_r {parent=dyn_capeBackShoulder_Base_r, loc=<0 0 0>, rot=<0 0 0 -1>}
 neck_01 {parent=spine_05, loc=<11.3138 0.1529 0>, rot=<0 0 -0.2003 -0.9797>}
 neck_02 {parent=neck_01, loc=<4.5627 0 0>, rot=<0 0 0.0504 -0.9987>}
 backpack_root {parent=spine_05, loc=<9.0735 -9.9874 0>, rot=<0.0284 -0.7065 -0.0284 -0.7065>}
 cape_mid_01_l {parent=backpack_root, loc=<6.2329 -0.2606 -0.4393>, rot=<0.1329 -0.6858 0.1296 -0.7037>}
 cape_mid_02_l {parent=cape_mid_01_l, loc=<12.8327 0 0>, rot=<0 0 -0.1478 -0.989>}
 cape_mid_03_l {parent=cape_mid_02_l, loc=<23.0451 0 0>, rot=<0 0 -0.0085 -1>}
 cape_mid_04_l {parent=cape_mid_03_l, loc=<20.3244 0 0>, rot=<0 0 -0.0139 -0.9999>}
 cape_mid_05_l {parent=cape_mid_04_l, loc=<18.7251 0 0>, rot=<0 0 0.0005 -1>}
 cape_mid_06_l {parent=cape_mid_05_l, loc=<13.6769 0 0>, rot=<0 0 0.0016 -1>}
 cape_mid_07_l {parent=cape_mid_06_l, loc=<14.7332 0 0>, rot=<0 0 -0.0176 -0.9998>}
 cape_mid_01_r {parent=backpack_root, loc=<-6.2329 -0.2606 -0.4389>, rot=<0.1296 -0.7037 0.1329 -0.6858>}
 cape_mid_02_r {parent=cape_mid_01_r, loc=<12.8333 -0.0003 0>, rot=<0 0 -0.1478 -0.989>}
 cape_mid_03_r {parent=cape_mid_02_r, loc=<23.0446 0 0>, rot=<0 0 -0.0085 -1>}
 cape_mid_04_r {parent=cape_mid_03_r, loc=<20.3246 0 0>, rot=<0 0 -0.0139 -0.9999>}
 cape_mid_05_r {parent=cape_mid_04_r, loc=<18.7251 0 0>, rot=<0 0 0.0005 -1>}
 cape_mid_06_r {parent=cape_mid_05_r, loc=<13.6769 -0.0001 0>, rot=<0 0 0.0016 -1>}
 cape_mid_07_r {parent=cape_mid_06_r, loc=<14.7332 0 0>, rot=<0 0 -0.0176 -0.9998>}
 dyn_capeChain_01_Base_l {parent=spine_05, loc=<-2.3345 14.456 -7.9749>, rot=<0.5447 -0.0278 0.8368 0.0486>}
 dyn_capeChain_01_l {parent=dyn_capeChain_01_Base_l, loc=<0 0 0>, rot=<0 0 0 -1>}
 dyn_capeChain_02_l {parent=dyn_capeChain_01_Base_l, loc=<2.8581 -0.4232 -0.7749>, rot=<-0.0007 0.0624 -0.0342 -0.9975>}
 dyn_capeChain_03_l {parent=dyn_capeChain_01_Base_l, loc=<5.7197 -0.28 -0.9202>, rot=<0.0019 0.1355 -0.041 -0.9899>}
 dyn_capeChain_01_Base_r {parent=spine_05, loc=<-2.4643 14.0139 8.2741>, rot=<-0.0739 -0.8358 0.0191 -0.5437>}
 dyn_capeChain_01_r {parent=dyn_capeChain_01_Base_r, loc=<0 0 0>, rot=<0 0 0 1>}
 dyn_capeChain_02_r {parent=dyn_capeChain_01_Base_r, loc=<2.9858 -0.1884 -0.6115>, rot=<0.0533 0.0533 -0.0133 -0.9971>}
 dyn_capeChain_03_r {parent=dyn_capeChain_01_Base_r, loc=<5.8947 -0.0678 -0.6026>, rot=<0.048 0.1031 -0.0066 -0.9935>}
 dyn_capeChainStart {parent=spine_05, loc=<-2.3994 14.2349 0.1496>, rot=<0.6607 -0.0747 0.0599 -0.7445>}
 dyn_capeChainEnd {parent=dyn_capeChainStart, loc=<-3.6252 0 0>, rot=<0 0 0 -1>}
]
variables [isSkydiving {type=double, default=0}]
