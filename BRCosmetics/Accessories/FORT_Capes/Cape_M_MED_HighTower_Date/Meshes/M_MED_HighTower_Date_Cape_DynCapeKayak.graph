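# Physics asset for the HighTower Date cape cosmetic, defining the bodies and constraints for its dynamic cape motion on the medium male character.
# Four chains of box bodies hang from the clavicle_cape and backpack_root bones, with limited swing and mostly locked twist.
# Angularly free constraints cross-link neighboring chains into a sheet from the upper segments down.
# Capsules and spheres on the body, a large dyn_crouch_collision sphere and a flat root box provide collision shapes.
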PhysicsAsset /BRCosmetics/Accessories/FORT_Capes/Cape_M_MED_HighTower_Date/Meshes/M_MED_HighTower_Date_Cape_DynCapeKayak
bodies {
 backpack_root [{shapeType=Capsule, length=0.505, radius=0.505, rotation=<rot 0 -90 0>}]
 calf_l [{shapeType=Capsule, length=31.8125, radius=12.5363, center=<-31.4163 -2.6359 -0.0003>, rotation=<rot 90 0 -1.5716>}]
 calf_r [{shapeType=Capsule, length=31.8125, radius=12.5363, center=<31.4163 2.6359 0.0003>, rotation=<rot 90 0 -1.5716>}]
 cape_01_l [{shapeType=Box, extentX=7.25, extentY=1.9225, extentZ=16.6183, center=<3.2928 -0.0002 0>}]
 cape_01_r [{shapeType=Box, extentX=7.25, extentY=0.9, extentZ=14.25, center=<3.2928 -0.0002 0>}]
 cape_02_l [{shapeType=Box, extentX=11.8125, extentY=2.2799, extentZ=18.6875, center=<6.1705 -0.0001 0>}]
 cape_02_r [{shapeType=Box, extentX=13.375, extentY=2.6507, extentZ=18.6875, center=<6.1705 -0.0001 0>}]
 cape_03_l [{shapeType=Box, extentX=22.5625, extentY=3.5, extentZ=21.0625, center=<11.5978 0.0001 0>}]
 cape_03_r [{shapeType=Box, extentX=22.5625, extentY=3.5, extentZ=21.0625, center=<11.5978 0.0001 0>}]
 cape_04_l [{shapeType=Box, extentX=20.5625, extentY=2.6875, extentZ=21.0625, center=<9.223 0.0001 0>}]
 cape_04_r [{shapeType=Box, extentX=20.5625, extentY=2.6875, extentZ=21.0625, center=<9.223 0.0001 0>}]
 cape_05_l [{shapeType=Box, extentX=19.125, extentY=4.4375, extentZ=21.0625, center=<8.6769 0.0001 0>}]
 cape_05_r [{shapeType=Box, extentX=19.125, extentY=4.4375, extentZ=21.0625, center=<8.6769 0.0001 0>}]
 cape_06_l [{shapeType=Box, extentX=13.375, extentY=4.625, extentZ=21.4375, center=<6.1705 -0.0001 0>}]
 cape_06_r [{shapeType=Box, extentX=13.375, extentY=4.625, extentZ=21.4375, center=<6.1705 -0.0001 0>}]
 cape_07_l [{shapeType=Box, extentX=14.625, extentY=2.6875, extentZ=20.125, center=<6.713 -0.0001 0>}]
 cape_07_r [{shapeType=Box, extentX=14.625, extentY=2.6875, extentZ=20.125, center=<6.713 -0.0001 0>}]
 cape_08_l [{shapeType=Box, extentX=25.0625, extentY=2.6875, extentZ=21.4375, center=<12.5115 0.0002 0>}]
 cape_08_r [{shapeType=Box, extentX=25.0625, extentY=2.6875, extentZ=21.4375, center=<12.5115 0.0002 0>}]
 cape_mid_01_l [{shapeType=Box, extentX=13.375, extentY=0.7, extentZ=18.6875, center=<6.1705 -0.0001 0>}]
 cape_mid_01_r [{shapeType=Box, extentX=13.375, extentY=2.6415, extentZ=18.6875, center=<6.1705 0.2867 0>}]
 cape_mid_02_l [{shapeType=Box, extentX=22.5625, extentY=1.625, extentZ=21.0625, center=<11.5978 0.0001 0>}]
 cape_mid_02_r [{shapeType=Box, extentX=22.5625, extentY=1.625, extentZ=21.0625, center=<11.5978 0.0001 0>}]
 cape_mid_03_l [{shapeType=Box, extentX=20.5625, extentY=2.6875, extentZ=21.0625, center=<9.223 0.0001 0>}]
 cape_mid_03_r [{shapeType=Box, extentX=20.5625, extentY=2.6875, extentZ=21.0625, center=<9.223 0.0001 0>}]
 cape_mid_04_l [{shapeType=Box, extentX=19.125, extentY=4.4375, extentZ=21.0625, center=<8.6769 0.0001 0>}]
 cape_mid_04_r [{shapeType=Box, extentX=19.125, extentY=4.4375, extentZ=21.0625, center=<8.6769 0.0001 0>}]
 cape_mid_05_l [{shapeType=Box, extentX=13.375, extentY=4.625, extentZ=21.4375, center=<6.1705 -0.0001 0>}]
 cape_mid_05_r [{shapeType=Box, extentX=13.375, extentY=4.625, extentZ=21.4375, center=<6.1705 -0.0001 0>}]
 cape_mid_06_l [{shapeType=Box, extentX=14.625, extentY=2.6875, extentZ=20.125, center=<6.713 -0.0001 0>}]
 cape_mid_06_r [{shapeType=Box, extentX=14.625, extentY=2.6875, extentZ=20.125, center=<6.713 -0.0001 0>}]
 cape_mid_07_l [{shapeType=Box, extentX=25.0625, extentY=2.6875, extentZ=21.4375, center=<12.5115 0.0002 0>}]
 cape_mid_07_r [{shapeType=Box, extentX=25.0625, extentY=2.6875, extentZ=21.4375, center=<12.5115 0.0002 0>}]
 clavicle_cape_l [{shapeType=Capsule, length=0.505, radius=0.505, rotation=<rot 0 -90 0>}]
 clavicle_cape_r [{shapeType=Capsule, length=0.505, radius=0.505, rotation=<rot 0 -90 0>}]
 clavicle_l [{shapeType=Capsule, length=0.505, radius=0.505, rotation=<rot 0 -90 0>}]
 clavicle_r [{shapeType=Capsule, length=0.505, radius=0.505, rotation=<rot 0 -90 0>}]
 dyn_crouch_collision [{shapeType=Sphere, radius=51.125, center=<-0.0006 6.727 0>}]
 lowerarm_l [{shapeType=Capsule, length=34.1377, radius=8.13, center=<18.227 0.9178 -2.1609>, rotation=<rot 82.7121 -0.0002 3.2014>}]
 lowerarm_r [
  {shapeType=Capsule, length=33.2939, radius=8.13, center=<-14.4192 -1.1325 1.674>, rotation=<rot 82.7121 -0.0002 3.2014>}
  {shapeType=Capsule, length=23.5938, radius=6.4062, center=<-31.9441 -1.2158 5.2412>, rotation=<rot 0 0 89.9994>}
 ]
 pelvis [
  {shapeType=Sphere, radius=15}
  {shapeType=Capsule, length=9.8125, radius=18.0363, center=<7.0434 -3.0718 18.0259>}
  {shapeType=Capsule, length=17.6562, radius=16.7238, center=<-21.9244 0 0.0008>}
 ]
 root [{shapeType=Box, extentX=231.0625, extentY=231.0625, extentZ=12.375, center=<0 0 -4.1865>}]
 spine_01 [{shapeType=Capsule, length=16.125, radius=11.7238}]
 spine_02 [{shapeType=Capsule, length=16.125, radius=11.7238}]
 spine_03 [{shapeType=Capsule, length=16.125, radius=12.25}]
 spine_04 [{shapeType=Capsule, length=16.125, radius=12.25}]
 spine_05 [{shapeType=Capsule, length=16.125, radius=12}]
 thigh_l [{shapeType=Capsule, length=16.125, radius=18.5988, center=<-20.2154 -0.6265 3.1209>, rotation=<rot 90 0 0>}]
 thigh_r [{shapeType=Capsule, length=16.125, radius=18.5988, center=<20.2154 0.6265 -2.7018>, rotation=<rot 90 0 0>}]
 upperarm_l [{shapeType=Capsule, length=23.5772, radius=10.3916, center=<13.1767 -0.475 -0.0001>, rotation=<rot 90 0 0>}]
 upperarm_r [{shapeType=Capsule, length=20.4279, radius=9.0408, center=<-13.1767 0.6421 0.0001>, rotation=<rot 90 0 0>}]
}
constraints [
 {bone1Name=cape_02_l, bone2Name=cape_01_l, swing1LimitDegrees=45, swing2LimitDegrees=25, twistLimitDegrees=5}
 {bone1Name=cape_03_l, bone2Name=cape_02_l, swing1LimitDegrees=45, swing2LimitDegrees=25, twistLimitDegrees=5, twistMotion=Locked}
 {bone1Name=cape_04_l, bone2Name=cape_03_l, swing1LimitDegrees=45, swing2LimitDegrees=25, twistLimitDegrees=5, twistMotion=Locked}
 {bone1Name=cape_05_l, bone2Name=cape_04_l, swing1LimitDegrees=45, swing2LimitDegrees=25, twistLimitDegrees=5, twistMotion=Locked}
 {bone1Name=cape_06_l, bone2Name=cape_05_l, swing1LimitDegrees=45, swing2LimitDegrees=25, twistLimitDegrees=5, twistMotion=Locked}
 {bone1Name=cape_07_l, bone2Name=cape_06_l, swing1LimitDegrees=45, swing2LimitDegrees=25, twistLimitDegrees=5, twistMotion=Locked}
 {bone1Name=cape_08_l, bone2Name=cape_07_l, swing1LimitDegrees=45, swing2LimitDegrees=25, twistLimitDegrees=5, twistMotion=Locked}
 {bone1Name=cape_01_l, bone2Name=clavicle_cape_l, swing1LimitDegrees=45, swing2LimitDegrees=45, twistLimitDegrees=5}
 {bone1Name=cape_01_r, bone2Name=clavicle_cape_r, swing1LimitDegrees=45, swing2LimitDegrees=45, twistLimitDegrees=5}
 {bone1Name=cape_02_r, bone2Name=cape_01_r, swing1LimitDegrees=45, swing2LimitDegrees=25, twistLimitDegrees=5}
 {bone1Name=cape_03_r, bone2Name=cape_02_r, swing1LimitDegrees=45, swing2LimitDegrees=25, twistLimitDegrees=5, twistMotion=Locked}
 {bone1Name=cape_04_r, bone2Name=cape_03_r, swing1LimitDegrees=45, swing2LimitDegrees=25, twistLimitDegrees=5, twistMotion=Locked}
 {bone1Name=cape_05_r, bone2Name=cape_04_r, swing1LimitDegrees=45, swing2LimitDegrees=25, twistLimitDegrees=5, twistMotion=Locked}
 {bone1Name=cape_06_r, bone2Name=cape_05_r, swing1LimitDegrees=45, swing2LimitDegrees=25, twistLimitDegrees=5, twistMotion=Locked}
 {bone1Name=cape_07_r, bone2Name=cape_06_r, swing1LimitDegrees=45, swing2LimitDegrees=25, twistLimitDegrees=5, twistMotion=Locked}
 {bone1Name=cape_08_r, bone2Name=cape_07_r, swing1LimitDegrees=45, swing2LimitDegrees=25, twistLimitDegrees=5, twistMotion=Locked}
 {bone1Name=cape_mid_01_l, bone2Name=backpack_root, swing1LimitDegrees=45, swing2LimitDegrees=45, twistLimitDegrees=5}
 {bone1Name=cape_mid_02_l, bone2Name=cape_mid_01_l, swing1LimitDegrees=45, swing2LimitDegrees=25, twistLimitDegrees=5, twistMotion=Locked}
 {bone1Name=cape_mid_03_l, bone2Name=cape_mid_02_l, swing1LimitDegrees=45, swing2LimitDegrees=25, twistLimitDegrees=5, twistMotion=Locked}
 {bone1Name=cape_mid_04_l, bone2Name=cape_mid_03_l, swing1LimitDegrees=45, swing2LimitDegrees=25, twistLimitDegrees=5, twistMotion=Locked}
 {bone1Name=cape_mid_05_l, bone2Name=cape_mid_04_l, swing1LimitDegrees=45, swing2LimitDegrees=25, twistLimitDegrees=5, twistMotion=Locked}
 {bone1Name=cape_mid_06_l, bone2Name=cape_mid_05_l, swing1LimitDegrees=45, swing2LimitDegrees=25, twistLimitDegrees=5, twistMotion=Locked}
 {bone1Name=cape_mid_07_l, bone2Name=cape_mid_06_l, swing1LimitDegrees=45, swing2LimitDegrees=25, twistLimitDegrees=5, twistMotion=Locked}
 {bone1Name=cape_mid_01_r, bone2Name=backpack_root, swing1LimitDegrees=45, swing2LimitDegrees=45, twistLimitDegrees=5}
 {bone1Name=cape_mid_02_r, bone2Name=cape_mid_01_r, swing1LimitDegrees=45, swing2LimitDegrees=25, twistLimitDegrees=5, twistMotion=Locked}
 {bone1Name=cape_mid_03_r, bone2Name=cape_mid_02_r, swing1LimitDegrees=45, swing2LimitDegrees=25, twistLimitDegrees=5, twistMotion=Locked}
 {bone1Name=cape_mid_04_r, bone2Name=cape_mid_03_r, swing1LimitDegrees=45, swing2LimitDegrees=25, twistLimitDegrees=5, twistMotion=Locked}
 {bone1Name=cape_mid_05_r, bone2Name=cape_mid_04_r, swing1LimitDegrees=45, swing2LimitDegrees=25, twistLimitDegrees=5, twistMotion=Locked}
 {bone1Name=cape_mid_06_r, bone2Name=cape_mid_05_r, swing1LimitDegrees=45, swing2LimitDegrees=25, twistLimitDegrees=5, twistMotion=Locked}
 {bone1Name=cape_mid_07_r, bone2Name=cape_mid_06_r, swing1LimitDegrees=45, swing2LimitDegrees=25, twistLimitDegrees=5, twistMotion=Locked}
 {
  bone1Name=clavicle_cape_l
  bone2Name=clavicle_l
  swing1LimitDegrees=45
  swing1Motion=Free
  swing2LimitDegrees=45
  swing2Motion=Free
  twistLimitDegrees=45
  twistMotion=Free
 }
 {
  bone1Name=clavicle_cape_r
  bone2Name=clavicle_r
  swing1LimitDegrees=45
  swing1Motion=Free
  swing2LimitDegrees=45
  swing2Motion=Free
  twistLimitDegrees=45
  twistMotion=Free
 }
 {
  bone1Name=clavicle_l
  bone2Name=spine_05
  swing1LimitDegrees=45
  swing1Motion=Free
  swing2LimitDegrees=45
  swing2Motion=Free
  twistLimitDegrees=45
  twistMotion=Free
 }
 {
  bone1Name=clavicle_r
  bone2Name=spine_05
  swing1LimitDegrees=45
  swing1Motion=Free
  swing2LimitDegrees=45
  swing2Motion=Free
  twistLimitDegrees=45
  twistMotion=Free
 }
 {
  bone1Name=backpack_root
  bone2Name=spine_05
  swing1LimitDegrees=45
  swing1Motion=Free
  swing2LimitDegrees=45
  swing2Motion=Free
  twistLimitDegrees=45
  twistMotion=Free
 }
 {
  bone1Name=spine_05
  bone2Name=spine_04
  swing1LimitDegrees=45
  swing1Motion=Free
  swing2LimitDegrees=45
  swing2Motion=Free
  twistLimitDegrees=45
  twistMotion=Free
 }
 {
  bone1Name=spine_04
  bone2Name=spine_03
  swing1LimitDegrees=45
  swing1Motion=Free
  swing2LimitDegrees=45
  swing2Motion=Free
  twistLimitDegrees=45
  twistMotion=Free
 }
 {
  bone1Name=spine_03
  bone2Name=spine_02
  swing1LimitDegrees=45
  swing1Motion=Free
  swing2LimitDegrees=45
  swing2Motion=Free
  twistLimitDegrees=45
  twistMotion=Free
 }
 {
  bone1Name=spine_02
  bone2Name=spine_01
  swing1LimitDegrees=45
  swing1Motion=Free
  swing2LimitDegrees=45
  swing2Motion=Free
  twistLimitDegrees=45
  twistMotion=Free
 }
 {
  bone1Name=spine_01
  bone2Name=pelvis
  swing1LimitDegrees=45
  swing1Motion=Free
  swing2LimitDegrees=45
  swing2Motion=Free
  twistLimitDegrees=45
  twistMotion=Free
 }
 {
  bone1Name=lowerarm_l
  bone2Name=upperarm_l
  swing1LimitDegrees=45
  swing1Motion=Free
  swing2LimitDegrees=45
  swing2Motion=Free
  twistLimitDegrees=45
  twistMotion=Free
 }
 {
  bone1Name=upperarm_l
  bone2Name=clavicle_l
  swing1LimitDegrees=45
  swing1Motion=Free
  swing2LimitDegrees=45
  swing2Motion=Free
  twistLimitDegrees=45
  twistMotion=Free
 }
 {
  bone1Name=lowerarm_r
  bone2Name=upperarm_r
  swing1LimitDegrees=45
  swing1Motion=Free
  swing2LimitDegrees=45
  swing2Motion=Free
  twistLimitDegrees=45
  twistMotion=Free
 }
 {
  bone1Name=upperarm_r
  bone2Name=clavicle_r
  swing1LimitDegrees=45
  swing1Motion=Free
  swing2LimitDegrees=45
  swing2Motion=Free
  twistLimitDegrees=45
  twistMotion=Free
 }
 {
  bone1Name=thigh_l
  bone2Name=pelvis
  swing1LimitDegrees=45
  swing1Motion=Free
  swing2LimitDegrees=45
  swing2Motion=Free
  twistLimitDegrees=45
  twistMotion=Free
 }
 {
  bone1Name=calf_l
  bone2Name=thigh_l
  swing1LimitDegrees=45
  swing1Motion=Free
  swing2LimitDegrees=45
  swing2Motion=Free
  twistLimitDegrees=45
  twistMotion=Free
 }
 {
  bone1Name=thigh_r
  bone2Name=pelvis
  swing1LimitDegrees=45
  swing1Motion=Free
  swing2LimitDegrees=45
  swing2Motion=Free
  twistLimitDegrees=45
  twistMotion=Free
 }
 {
  bone1Name=calf_r
  bone2Name=thigh_r
  swing1LimitDegrees=45
  swing1Motion=Free
  swing2LimitDegrees=45
  swing2Motion=Free
  twistLimitDegrees=45
  twistMotion=Free
 }
 {
  bone1Name=pelvis
  bone2Name=root
  swing1LimitDegrees=45
  swing1Motion=Free
  swing2LimitDegrees=45
  swing2Motion=Free
  twistLimitDegrees=45
  twistMotion=Free
 }
 {
  bone1Name=dyn_crouch_collision
  bone2Name=root
  swing1LimitDegrees=45
  swing1Motion=Free
  swing2LimitDegrees=45
  swing2Motion=Free
  twistLimitDegrees=45
  twistMotion=Free
 }
 {
  bone1Name=cape_03_l
  bone2Name=cape_mid_02_l
  swing1LimitDegrees=45
  swing1Motion=Free
  swing2LimitDegrees=45
  swing2Motion=Free
  twistLimitDegrees=45
  twistMotion=Free
 }
 {
  bone1Name=cape_04_l
  bone2Name=cape_mid_03_l
  swing1LimitDegrees=45
  swing1Motion=Free
  swing2LimitDegrees=45
  swing2Motion=Free
  twistLimitDegrees=45
  twistMotion=Free
 }
 {
  bone1Name=cape_05_l
  bone2Name=cape_mid_04_l
  swing1LimitDegrees=45
  swing1Motion=Free
  swing2LimitDegrees=45
  swing2Motion=Free
  twistLimitDegrees=45
  twistMotion=Free
 }
 {
  bone1Name=cape_06_l
  bone2Name=cape_mid_05_l
  swing1LimitDegrees=45
  swing1Motion=Free
  swing2LimitDegrees=45
  swing2Motion=Free
  twistLimitDegrees=45
  twistMotion=Free
 }
 {
  bone1Name=cape_07_l
  bone2Name=cape_mid_06_l
  swing1LimitDegrees=45
  swing1Motion=Free
  swing2LimitDegrees=45
  swing2Motion=Free
  twistLimitDegrees=45
  twistMotion=Free
 }
 {
  bone1Name=cape_08_l
  bone2Name=cape_mid_07_l
  swing1LimitDegrees=45
  swing1Motion=Free
  swing2LimitDegrees=45
  swing2Motion=Free
  twistLimitDegrees=45
  twistMotion=Free
 }
 {
  bone1Name=cape_03_r
  bone2Name=cape_mid_02_r
  swing1LimitDegrees=45
  swing1Motion=Free
  swing2LimitDegrees=45
  swing2Motion=Free
  twistLimitDegrees=45
  twistMotion=Free
 }
 {
  bone1Name=cape_04_r
  bone2Name=cape_mid_03_r
  swing1LimitDegrees=45
  swing1Motion=Free
  swing2LimitDegrees=45
  swing2Motion=Free
  twistLimitDegrees=45
  twistMotion=Free
 }
 {
  bone1Name=cape_05_r
  bone2Name=cape_mid_04_r
  swing1LimitDegrees=45
  swing1Motion=Free
  swing2LimitDegrees=45
  swing2Motion=Free
  twistLimitDegrees=45
  twistMotion=Free
 }
 {
  bone1Name=cape_06_r
  bone2Name=cape_mid_05_r
  swing1LimitDegrees=45
  swing1Motion=Free
  swing2LimitDegrees=45
  swing2Motion=Free
  twistLimitDegrees=45
  twistMotion=Free
 }
 {
  bone1Name=cape_07_r
  bone2Name=cape_mid_06_r
  swing1LimitDegrees=45
  swing1Motion=Free
  swing2LimitDegrees=45
  swing2Motion=Free
  twistLimitDegrees=45
  twistMotion=Free
 }
 {
  bone1Name=cape_08_r
  bone2Name=cape_mid_07_r
  swing1LimitDegrees=45
  swing1Motion=Free
  swing2LimitDegrees=45
  swing2Motion=Free
  twistLimitDegrees=45
  twistMotion=Free
 }
 {
  bone1Name=cape_mid_02_r
  bone2Name=cape_mid_02_l
  swing1LimitDegrees=45
  swing1Motion=Free
  swing2LimitDegrees=45
  swing2Motion=Free
  twistLimitDegrees=45
  twistMotion=Free
 }
 {
  bone1Name=cape_mid_03_l
  bone2Name=cape_mid_03_r
  swing1LimitDegrees=45
  swing1Motion=Free
  swing2LimitDegrees=45
  swing2Motion=Free
  twistLimitDegrees=45
  twistMotion=Free
 }
 {
  bone1Name=cape_mid_04_r
  bone2Name=cape_mid_04_l
  swing1LimitDegrees=45
  swing1Motion=Free
  swing2LimitDegrees=45
  swing2Motion=Free
  twistLimitDegrees=45
  twistMotion=Free
 }
 {
  bone1Name=cape_mid_05_l
  bone2Name=cape_mid_05_r
  swing1LimitDegrees=45
  swing1Motion=Free
  swing2LimitDegrees=45
  swing2Motion=Free
  twistLimitDegrees=45
  twistMotion=Free
 }
 {
  bone1Name=cape_mid_06_r
  bone2Name=cape_mid_06_l
  swing1LimitDegrees=45
  swing1Motion=Free
  swing2LimitDegrees=45
  swing2Motion=Free
  twistLimitDegrees=45
  twistMotion=Free
 }
 {
  bone1Name=cape_mid_07_l
  bone2Name=cape_mid_07_r
  swing1LimitDegrees=45
  swing1Motion=Free
  swing2LimitDegrees=45
  swing2Motion=Free
  twistLimitDegrees=45
  twistMotion=Free
 }
]
