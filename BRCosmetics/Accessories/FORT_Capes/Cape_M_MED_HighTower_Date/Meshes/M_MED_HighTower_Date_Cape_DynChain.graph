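# Physics asset for the chain on the HighTower Date cape accessory for the medium male body, defining collision shapes and joint limits for its dynamic bone.
# A small box body covers the dyn_capeChainStart bone, and a capsule body covers the spine_05 upper-spine bone.
# A single constraint joins the two bodies, with limited swing on both axes and a much tighter twist range.
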PhysicsAsset /BRCosmetics/Accessories/FORT_Capes/Cape_M_MED_HighTower_Date/Meshes/M_MED_HighTower_Date_Cape_DynChain
bodies {
 dyn_capeChainStart [{shapeType=Box, extentX=4.0193, extentY=6.0119, extentZ=1.01, center=<-2.5694 0 0>, rotation=<rot 0 -90 0>}]
 spine_05 [{shapeType=Capsule, length=19.8148, radius=9.1257, center=<-2.1391 3.6455 0.1953>, rotation=<rot -0.3706 -102.4668 -89.6888>}]
}
constraints [{bone1Name=dyn_capeChainStart, bone2Name=spine_05, swing1LimitDegrees=50, swing2LimitDegrees=50, twistLimitDegrees=8}]
