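# Physics asset for the alternate hood-down dynamic bones of the Galileo Kayak cape accessory, giving the lowered hood secondary motion.
# Left, right and two-segment middle chains each hang from a small capsule MSTR root, using thin boxes for sides and capsules for the middle.
# Side flaps swing widely on one axis, while the middle chain bends narrowly with root twist locked.
# Angularly free constraints also tie both flaps to the first middle segment.
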PhysicsAsset /BRCosmetics/Accessories/FORT_Capes/M_MED_Galileo_Kayak_Cape/Meshes/M_MED_Galileo_Kayak_Pack_Dyn_hood_down_alt
bodies {
 dyn_downhood_alt_l_01 [{shapeType=Box, extentX=5.6875, extentY=0.7, extentZ=13.5, center=<2.1734 -0.4844 0.0001>, rotation=<rot 0.0001 -12.9903 0>}]
 dyn_downhood_alt_l_MSTR [{shapeType=Capsule, length=0.505, radius=0.505, rotation=<rot 0 -90 0>}]
 dyn_downhood_alt_mid_01 [{shapeType=Capsule, length=10, radius=3.5, shapeName=dyn_downhood_alt_mid_01_capsule, center=<5 0 0>, rotation=<rot 0 0 90>}]
 dyn_downhood_alt_mid_02 [{shapeType=Capsule, length=3, radius=1, shapeName=dyn_downhood_alt_mid_01_capsule, center=<5 0 0>, rotation=<rot 0 0 90>}]
 dyn_downhood_alt_mid_MSTR [{shapeType=Capsule, length=0.505, radius=0.505, rotation=<rot 0 -90 0>}]
 dyn_downhood_alt_r_01 [{shapeType=Box, extentX=5.6875, extentY=0.7, extentZ=13.5, center=<2.1734 0.4844 -0.0001>, rotation=<rot 0.0001 12.9903 0>}]
 dyn_downhood_alt_r_MSTR [{shapeType=Capsule, length=0.505, radius=0.505, rotation=<rot 0 -90 0>}]
}
constraints [
 {bone1Name=dyn_downhood_alt_l_01, bone2Name=dyn_downhood_alt_l_MSTR, swing1LimitDegrees=90, swing2LimitDegrees=5, twistLimitDegrees=10}
 {bone1Name=dyn_downhood_alt_r_01, bone2Name=dyn_downhood_alt_r_MSTR, swing1LimitDegrees=90, swing2LimitDegrees=5, twistLimitDegrees=10}
 {
  bone1Name=dyn_downhood_alt_mid_01
  bone2Name=dyn_downhood_alt_mid_MSTR
  swing1LimitDegrees=15
  swing2LimitDegrees=15
  twistLimitDegrees=45
  twistMotion=Locked
 }
 {bone1Name=dyn_downhood_alt_mid_02, bone2Name=dyn_downhood_alt_mid_01, swing1LimitDegrees=15, swing2LimitDegrees=15, twistLimitDegrees=15}
 {
  bone1Name=dyn_downhood_alt_r_01
  bone2Name=dyn_downhood_alt_mid_01
  swing1LimitDegrees=45
  swing1Motion=Free
  swing2LimitDegrees=45
  swing2Motion=Free
  twistLimitDegrees=45
  twistMotion=Free
 }
 {
  bone1Name=dyn_downhood_alt_l_01
  bone2Name=dyn_downhood_alt_mid_01
  swing1LimitDegrees=45
  swing1Motion=Free
  swing2LimitDegrees=45
  swing2Motion=Free
  twistLimitDegrees=45
  twistMotion=Free
 }
]
