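# Physics asset for the hood-up variant of the Galileo Kayak cape cosmetic, defining collision and joints for its dynamic hood bones.
# Three capsule chains (top, left and right) each start at a MSTR bone constrained to head_alt with free swing and twist.
# Joints down each chain have tight angular limits of a few degrees, allowing little bend between segments.
# Two-capsule head and head_alt bodies and tiny clavicle capsules complete the body set.
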PhysicsAsset /BRCosmetics/Accessories/FORT_Capes/M_MED_Galileo_Kayak_Cape/Meshes/M_MED_Galileo_Kayak_Pack_Dyn_hood_up
bodies {
 clavicle_l [{shapeType=Capsule, length=0.505, radius=0.505, rotation=<rot 0 -90 0>}]
 clavicle_r [{shapeType=Capsule, length=0.505, radius=0.505, rotation=<rot 0 -90 0>}]
 dyn_hood_simple_l_01 [{shapeType=Capsule, length=3, radius=3, shapeName=dyn_hood_simple_l_01_capsule}]
 dyn_hood_simple_l_02 [{shapeType=Capsule, length=2.5, radius=3.5, shapeName=dyn_hood_simple_l_02_capsule, center=<3 0 0>, rotation=<rot 90 0 0>}]
 dyn_hood_simple_l_03 [{shapeType=Capsule, length=2.5, radius=3.5, shapeName=dyn_hood_simple_l_03_capsule, center=<3 0 0>, rotation=<rot 90 0 0>}]
 dyn_hood_simple_l_MSTR [{shapeType=Capsule, length=0.505, radius=0.505, shapeName=dyn_hood_simple_l_MSTR_capsule, rotation=<rot 0 -90 0>}]
 dyn_hood_simple_r_01 [{shapeType=Capsule, length=3, radius=3, shapeName=dyn_hood_simple_r_01_capsule}]
 dyn_hood_simple_r_02 [{shapeType=Capsule, length=2.5, radius=3.5, shapeName=dyn_hood_simple_r_02_capsule, center=<3 0 0>, rotation=<rot 90 0 0>}]
 dyn_hood_simple_r_03 [{shapeType=Capsule, length=2.5, radius=3.5, shapeName=dyn_hood_simple_r_03_capsule, center=<3 0 0>, rotation=<rot 90 0 0>}]
 dyn_hood_simple_r_MSTR [{shapeType=Capsule, length=0.505, radius=0.505, shapeName=dyn_hood_simple_r_MSTR_capsule, rotation=<rot 0 -90 0>}]
 dyn_hood_top_01 [{shapeType=Capsule, length=8.5, radius=2.25, shapeName=dyn_hood_simple_r_02_capsule, center=<3 0 2.7983>, rotation=<rot 90 0 0>}]
 dyn_hood_top_02 [{shapeType=Capsule, length=6.75, radius=1.75, shapeName=dyn_hood_top_02_capsule, center=<3.0865 0 2.1497>, rotation=<rot 90 0 0>}]
 dyn_hood_top_MSTR [{shapeType=Capsule, length=0.505, radius=0.505, rotation=<rot 0 -90 0>}]
 head [
  {shapeType=Capsule, length=0, radius=12.2188, center=<0.8248 0.7659 -0.0002>, rotation=<rot 90 90 65.558>}
  {shapeType=Capsule, length=9.9062, radius=9.6875, center=<-0.4855 5.9457 0>, rotation=<rot 90 0 -15.9001>}
 ]
 head_alt [
  {shapeType=Capsule, length=3.0625, radius=10.3438, center=<3.5393 3.8753 -0.0005>, rotation=<rot 90 180 155.5548>}
  {shapeType=Capsule, length=9.9062, radius=9.6875, center=<-0.4855 5.9457 0>, rotation=<rot 90 0 -15.9001>}
 ]
}
constraints [
 {
  bone1Name=dyn_hood_top_MSTR
  bone2Name=head_alt
  swing1LimitDegrees=45
  swing1Motion=Free
  swing2LimitDegrees=45
  swing2Motion=Free
  twistLimitDegrees=45
  twistMotion=Free
 }
 {bone1Name=dyn_hood_top_01, bone2Name=dyn_hood_top_MSTR, swing1LimitDegrees=2, swing2LimitDegrees=5, twistLimitDegrees=2}
 {bone1Name=dyn_hood_top_02, bone2Name=dyn_hood_top_01, swing1LimitDegrees=2, swing2LimitDegrees=5, twistLimitDegrees=2}
 {
  bone1Name=dyn_hood_simple_l_MSTR
  bone2Name=head_alt
  swing1LimitDegrees=45
  swing1Motion=Free
  swing2LimitDegrees=45
  swing2Motion=Free
  twistLimitDegrees=45
  twistMotion=Free
 }
 {bone1Name=dyn_hood_simple_l_01, bone2Name=dyn_hood_simple_l_MSTR, swing1LimitDegrees=5, swing2LimitDegrees=5, twistLimitDegrees=5}
 {bone1Name=dyn_hood_simple_l_02, bone2Name=dyn_hood_simple_l_01, swing1LimitDegrees=5, swing2LimitDegrees=5, twistLimitDegrees=2}
 {bone1Name=dyn_hood_simple_l_03, bone2Name=dyn_hood_simple_l_02, swing1LimitDegrees=5, swing2LimitDegrees=5, twistLimitDegrees=2}
 {
  bone1Name=dyn_hood_simple_r_MSTR
  bone2Name=head_alt
  swing1LimitDegrees=45
  swing1Motion=Free
  swing2LimitDegrees=45
  swing2Motion=Free
  twistLimitDegrees=45
  twistMotion=Free
 }
 {bone1Name=dyn_hood_simple_r_01, bone2Name=dyn_hood_simple_r_MSTR, swing1LimitDegrees=5, swing2LimitDegrees=5, twistLimitDegrees=5}
 {bone1Name=dyn_hood_simple_r_02, bone2Name=dyn_hood_simple_r_01, swing1LimitDegrees=5, swing2LimitDegrees=5, twistLimitDegrees=2}
 {bone1Name=dyn_hood_simple_r_03, bone2Name=dyn_hood_simple_r_02, swing1LimitDegrees=5, swing2LimitDegrees=5, twistLimitDegrees=2}
]
