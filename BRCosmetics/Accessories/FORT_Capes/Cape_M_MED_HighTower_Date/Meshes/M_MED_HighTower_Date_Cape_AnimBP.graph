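# Animation Blueprint for the HighTower Date cape cosmetic, posing the cape from a pose copied from the attached parent mesh.
# Cape and chain RigidBody simulations run when enabled, otherwise Look At or bone offsets with Trail controllers substitute, split by DBNO state.
# The result feeds a skydive-aware Control Rig, then additive tweaks for sprint, skydive and physics-off.
# Events set crouch-collision alphas, reset dynamics around skydiving and detect the matching glider.
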AnimBlueprint /BRCosmetics/Accessories/FORT_Capes/Cape_M_MED_HighTower_Date/Meshes/M_MED_HighTower_Date_Cape_AnimBP
target_skeleton=&/M_MED_HighTower_Date_Cape_Skeleton
graphs [
 AnimGraph {nodes [
   n1 = RigidBody!(ComponentPose, Alpha, OverrideWorldGravity:Vector="0,0,0", ExternalForce) -> (Pose) {
    bClampLinearTranslationLimitToRefPose=true
    bOverrideWorldGravity=true
    overridePhysicsAsset=/BRCosmetics/Accessories/FORT_Capes/M_MED_Galileo_Kayak_Cape/Meshes/M_MED_Galileo_Kayak_Pack_Dyn_hood_up
    simulationSpace=BaseBoneSpace
    alphaBoolBlend {blendOption=Linear}
    baseBoneRef {boneName=neck_01}
    componentAppliedLinearAccClamp=<3000 3000 3000>
    componentLinearAccScale=<0.4 0.3 0.3>
    componentLinearVelScale=<0.4 0.3 0.3>
    overrideWorldGravity {z=-980}
   }
   RigidBody(ComponentPose=@n1, OverrideWorldGravity="0,0,0") {
    bClampLinearTranslationLimitToRefPose=true
    bOverrideWorldGravity=true
    overridePhysicsAsset=/BRCosmetics/Accessories/FORT_Capes/M_MED_Galileo_Kayak_Cape/Meshes/M_MED_Galileo_Kayak_Pack_Dyn_hood_down_alt
    simulationSpace=BaseBoneSpace
    alphaBoolBlend {blendOption=Linear}
    baseBoneRef {boneName=spine_05}
    componentAppliedLinearAccClamp=<3000 3000 3000>
    componentLinearAccScale=<0.4 0.3 0.3>
    componentLinearVelScale=<0.4 0.3 0.3>
    overrideWorldGravity {z=-980}
   }
   n2 = NOT Boolean(A=Get bRigidBodyEnabled())
   n3 = AND Boolean(A=@n2, B=Get bIsSprinting())
   n4 = Get bIsSkyDiving()
   n5 = Get bIsSprinting()
   n6 = Look At - Bone: cape_mid_02_l(ComponentPose=Local To Component(LocalPose=Use cached pose 'initialPoseCache'()), Alpha=1) {
    alphaBoolBlend {blendOption=Linear}
    boneToModify {boneName=cape_mid_02_l}
    lookAtLocation=<0 -40 -11>
    lookAtTarget {boneReference {boneName=pelvis}}
    lookAt_Axis {axis=<1 0>}
   }
   n7 = Look At - Bone: cape_mid_02_r(ComponentPose=@n6, Alpha=1) {
    alphaBoolBlend {blendOption=Linear}
    boneToModify {boneName=cape_mid_02_r}
    lookAtLocation=<0 -40 11>
    lookAtTarget {boneReference {boneName=pelvis}}
    lookAt_Axis {axis=<1 0>}
   }
   n8 = Look At - Bone: cape_02_l(ComponentPose=@n7, Alpha=1) {
    alphaBoolBlend {blendOption=Linear}
    boneToModify {boneName=cape_02_l}
    lookAtLocation=<0 -25 -21>
    lookAtTarget {boneReference {boneName=pelvis}}
    lookAt_Axis {axis=<1 0>}
   }
   n9 = Look At - Bone: cape_02_r(ComponentPose=@n8, Alpha=1) {
    alphaBoolBlend {blendOption=Linear}
    boneToModify {boneName=cape_02_r}
    lookAtLocation=<0 -25 21>
    lookAtTarget {boneReference {boneName=pelvis}}
    lookAt_Axis {axis=<1 0>}
   }
   n10 = Look At - Bone: cape_mid_05_l(ComponentPose=@n9, Alpha=1) {
    alphaBoolBlend {blendOption=Linear}
    boneToModify {boneName=cape_mid_05_l}
    lookAtLocation=<-75 -15 -11>
    lookAtTarget {boneReference {boneName=pelvis}}
    lookAt_Axis {axis=<1 0>}
   }
   n11 = Look At - Bone: cape_mid_05_r(ComponentPose=@n10, Alpha=1) {
    alphaBoolBlend {blendOption=Linear}
    boneToModify {boneName=cape_mid_05_r}
    lookAtLocation=<-75 -15 11>
    lookAtTarget {boneReference {boneName=pelvis}}
    lookAt_Axis {axis=<1 0>}
   }
   n12 = Look At - Bone: cape_06_l(ComponentPose=@n11, Alpha=1) {
    alphaBoolBlend {blendOption=Linear}
    boneToModify {boneName=cape_06_l}
    lookAtLocation=<-75 -8 -21>
    lookAtTarget {boneReference {boneName=pelvis}}
    lookAt_Axis {axis=<1 0>}
   }
   n13 = Look At - Bone: cape_06_r(ComponentPose=@n12, Alpha=1) {
    alphaBoolBlend {blendOption=Linear}
    boneToModify {boneName=cape_06_r}
    lookAtLocation=<-75 -8 21>
    lookAtTarget {boneReference {boneName=pelvis}}
    lookAt_Axis {axis=<1 0>}
   }
   n14 = Trail controller - Bone: cape_mid_07_l!(ComponentPose=@n13, Alpha=1, RelaxationSpeedScale:float=2, FakeVelocity:Vector="0,0,0") -> (Pose) {
    bActorSpaceFakeVel=true
    bShowTrailLocation=true
    bUsePlanarLimit=true
    chainLength=6
    debugLifeTime=0.1
    maxDeltaTime=0.0333
    relaxationSpeedScale=1.2
    alphaScaleBiasClamp {bInterpResult=true, interpSpeedDecreasing=25, interpSpeedIncreasing=25}
    baseJoint {boneName=root}
    rotationLimits [
     {limitMax=<180 180 180>}
     {limitMax=<180 180 180>}
     {limitMax=<180 180 180>}
     {limitMax=<180 180 180>}
     {limitMax=<180 180 180>}
     {limitMax=<180 180 180>}
    ]
    rotationOffsets [<0 0 0>, <0 0 0>, <0 0 0>, <0 0 0>, <0 0 0>, <0 0 0>]
    trailBone {boneName=cape_mid_07_l}
    trailRelaxationSpeed {editorCurveData {keys [<at 0 15 RCIM_Linear>,<at 1 6 RCIM_Linear>]}}
   }
   n15 = Trail controller - Bone: cape_mid_07_r!(ComponentPose=@n14, Alpha=1, RelaxationSpeedScale:float=2, FakeVelocity:Vector="0,0,0") -> (Pose) {
    bActorSpaceFakeVel=true
    bShowTrailLocation=true
    bUsePlanarLimit=true
    chainLength=6
    debugLifeTime=0.1
    maxDeltaTime=0.0333
    relaxationSpeedScale=1.2
    alphaScaleBiasClamp {bInterpResult=true, interpSpeedDecreasing=25, interpSpeedIncreasing=25}
    baseJoint {boneName=root}
    rotationLimits [
     {limitMax=<180 180 180>}
     {limitMax=<180 180 180>}
     {limitMax=<180 180 180>}
     {limitMax=<180 180 180>}
     {limitMax=<180 180 180>}
     {limitMax=<180 180 180>}
    ]
    rotationOffsets [<0 0 0>, <0 0 0>, <0 0 0>, <0 0 0>, <0 0 0>, <0 0 0>]
    trailBone {boneName=cape_mid_07_r}
    trailRelaxationSpeed {editorCurveData {keys [<at 0 15 RCIM_Linear>,<at 1 6 RCIM_Linear>]}}
   }
   n16 = Trail controller - Bone: cape_08_l!(ComponentPose=@n15, Alpha=1, RelaxationSpeedScale:float=2, FakeVelocity:Vector="0,0,0") -> (Pose) {
    bActorSpaceFakeVel=true
    bShowTrailLocation=true
    bUsePlanarLimit=true
    chainLength=6
    debugLifeTime=0.1
    maxDeltaTime=0.0333
    relaxationSpeedScale=1.2
    alphaScaleBiasClamp {bInterpResult=true, interpSpeedDecreasing=25, interpSpeedIncreasing=25}
    baseJoint {boneName=root}
    rotationLimits [
     {limitMax=<180 180 180>}
     {limitMax=<180 180 180>}
     {limitMax=<180 180 180>}
     {limitMax=<180 180 180>}
     {limitMax=<180 180 180>}
     {limitMax=<180 180 180>}
    ]
    rotationOffsets [<0 0 0>, <0 0 0>, <0 0 0>, <0 0 0>, <0 0 0>, <0 0 0>]
    trailBone {boneName=cape_08_l}
    trailRelaxationSpeed {editorCurveData {keys [<at 0 15 RCIM_Linear>,<at 1 6 RCIM_Linear>]}}
   }
   n17 = Trail controller - Bone: cape_08_r!(ComponentPose=@n16, Alpha=1, RelaxationSpeedScale:float=2, FakeVelocity:Vector="0,0,0") -> (Pose) {
    bActorSpaceFakeVel=true
    bShowTrailLocation=true
    bUsePlanarLimit=true
    chainLength=6
    debugLifeTime=0.1
    maxDeltaTime=0.0333
    relaxationSpeedScale=1.2
    alphaScaleBiasClamp {bInterpResult=true, interpSpeedDecreasing=25, interpSpeedIncreasing=25}
    baseJoint {boneName=root}
    rotationLimits [
     {limitMax=<180 180 180>}
     {limitMax=<180 180 180>}
     {limitMax=<180 180 180>}
     {limitMax=<180 180 180>}
     {limitMax=<180 180 180>}
     {limitMax=<180 180 180>}
    ]
    rotationOffsets [<0 0 0>, <0 0 0>, <0 0 0>, <0 0 0>, <0 0 0>, <0 0 0>]
    trailBone {boneName=cape_08_r}
    trailRelaxationSpeed {editorCurveData {keys [<at 0 15 RCIM_Linear>,<at 1 6 RCIM_Linear>]}}
   }
   n18 = Component To Local(ComponentPose=@n17)
   n19 = Look At - Bone: cape_02_l(ComponentPose=Local To Component(LocalPose=Use cached pose 'initialPoseCache'()), Alpha=1) {
    alphaBoolBlend {blendOption=Linear}
    boneToModify {boneName=cape_02_l}
    lookAtLocation=<0 -30 -21>
    lookAtTarget {boneReference {boneName=pelvis}}
    lookAt_Axis {axis=<1 0>}
   }
   n20 = Look At - Bone: cape_02_r(ComponentPose=@n19, Alpha=1) {
    alphaBoolBlend {blendOption=Linear}
    boneToModify {boneName=cape_02_r}
    lookAtLocation=<0 -30 21>
    lookAtTarget {boneReference {boneName=pelvis}}
    lookAt_Axis {axis=<1 0>}
   }
   n21 = "Transform (Modify) Bone - Bone: cape_04_l"(ComponentPose=@n20, Alpha=1, Translation="0,0,0", Rotation="0,-40,0", Scale="1,1,1") {
    rotationMode=BMM_Additive
    rotationSpace=BCS_BoneSpace
    translationMode=BMM_Additive
    translationSpace=BCS_BoneSpace
    alphaBoolBlend {blendOption=Linear}
    alphaScaleBiasClamp {bInterpResult=true, interpSpeedDecreasing=4, interpSpeedIncreasing=4}
    boneToModify {boneName=cape_04_l}
    rotation {yaw=-40}
   }
   n22 = "Transform (Modify) Bone - Bone: cape_04_r"(ComponentPose=@n21, Alpha=1, Translation="0,0,0", Rotation="0,-40,0", Scale="1,1,1") {
    rotationMode=BMM_Additive
    rotationSpace=BCS_BoneSpace
    translationMode=BMM_Additive
    translationSpace=BCS_BoneSpace
    alphaBoolBlend {blendOption=Linear}
    alphaScaleBiasClamp {bInterpResult=true, interpSpeedDecreasing=4, interpSpeedIncreasing=4}
    boneToModify {boneName=cape_04_r}
    rotation {yaw=-40}
   }
   n23 = "Transform (Modify) Bone - Bone: cape_mid_03_l"(ComponentPose=@n22, Alpha=1, Translation="0,0,0", Rotation="0,-40,0", Scale="1,1,1") {
    rotationMode=BMM_Additive
    rotationSpace=BCS_BoneSpace
    translationMode=BMM_Additive
    translationSpace=BCS_BoneSpace
    alphaBoolBlend {blendOption=Linear}
    alphaScaleBiasClamp {bInterpResult=true, interpSpeedDecreasing=4, interpSpeedIncreasing=4}
    boneToModify {boneName=cape_mid_03_l}
    rotation {yaw=-40}
   }
   n24 = "Transform (Modify) Bone - Bone: cape_mid_03_r"(ComponentPose=@n23, Alpha=1, Translation="0,0,0", Rotation="0,-40,0", Scale="1,1,1") {
    rotationMode=BMM_Additive
    rotationSpace=BCS_BoneSpace
    translationMode=BMM_Additive
    translationSpace=BCS_BoneSpace
    alphaBoolBlend {blendOption=Linear}
    alphaScaleBiasClamp {bInterpResult=true, interpSpeedDecreasing=4, interpSpeedIncreasing=4}
    boneToModify {boneName=cape_mid_03_r}
    rotation {yaw=-40}
   }
   n25 = Get CapeTrailRelaxSpeedScale() -> (CapeTrailRelaxSpeedScale:float)
   n26 = Trail controller - Bone: cape_mid_07_l(
    ComponentPose=@n24
    Alpha=1
    RelaxationSpeedScale=@n25
    FakeVelocity=Get CapeTrailFakeVelocityMidLeft_b() -> (CapeTrailFakeVelocityMidLeft_b:Vector)
   ) {
    bActorSpaceFakeVel=true
    bShowTrailLocation=true
    bUsePlanarLimit=true
    chainLength=6
    debugLifeTime=0.1
    maxDeltaTime=0.0333
    relaxationSpeedScale=1.2
    alphaScaleBiasClamp {bInterpResult=true}
    baseJoint {boneName=spine_03}
    planarLimits [
     {drivingBone {boneName=root}, planeTransform {location=<0 0 10>, rotation=<rot 0 0 0>, scale=<1 1 1>}}
     {drivingBone {boneName=spine_03}, planeTransform {location=<0 -20 0>, rotation=<rot 0 15 -89.9999>, scale=<1 1 1>}}
    ]
    rotationLimits [
     {limitMax=<180 180 180>}
     {limitMax=<180 180 180>}
     {limitMax=<180 180 180>}
     {limitMax=<180 180 180>}
     {limitMax=<180 180 180>}
     {limitMax=<180 180 180>}
    ]
    rotationOffsets [<0 0 0>, <0 0 0>, <0 0 0>, <0 0 0>, <0 0 0>, <0 0 0>]
    trailBone {boneName=cape_mid_07_l}
    trailRelaxationSpeed {editorCurveData {keys [<at 0 15 RCIM_Linear>,<at 0.5281 8.315 RCIM_Linear>,<at 1 6 RCIM_Linear>]}}
   }
   n27 = Trail controller - Bone: cape_mid_07_r(
    ComponentPose=@n26
    Alpha=1
    RelaxationSpeedScale=@n25
    FakeVelocity=Get CapeTrailFakeVelocityMidRight_b() -> (CapeTrailFakeVelocityMidRight_b:Vector)
   ) {
    bActorSpaceFakeVel=true
    bShowTrailLocation=true
    bUsePlanarLimit=true
    chainLength=6
    debugLifeTime=0.1
    maxDeltaTime=0.0333
    relaxationSpeedScale=1.2
    alphaScaleBiasClamp {bInterpResult=true}
    baseJoint {boneName=spine_03}
    planarLimits [
     {drivingBone {boneName=root}, planeTransform {location=<0 0 10>, rotation=<rot 0 0 0>, scale=<1 1 1>}}
     {drivingBone {boneName=spine_03}, planeTransform {location=<0 -20 0>, rotation=<rot 0 15 -89.9999>, scale=<1 1 1>}}
    ]
    rotationLimits [
     {limitMax=<180 180 180>}
     {limitMax=<180 180 180>}
     {limitMax=<180 180 180>}
     {limitMax=<180 180 180>}
     {limitMax=<180 180 180>}
     {limitMax=<180 180 180>}
    ]
    rotationOffsets [<0 0 0>, <0 0 0>, <0 0 0>, <0 0 0>, <0 0 0>, <0 0 0>]
    trailBone {boneName=cape_mid_07_r}
    trailRelaxationSpeed {editorCurveData {keys [<at 0 15 RCIM_Linear>,<at 0.5281 8.315 RCIM_Linear>,<at 1 6 RCIM_Linear>]}}
   }
   n28 = Get CapeTrailRelaxSpeedScale()
   n29 = Trail controller - Bone: cape_08_l(
    ComponentPose=@n27
    Alpha=1
    RelaxationSpeedScale=@n28
    FakeVelocity=Get CapeTrailFakeVelocityLowLeft_b() -> (CapeTrailFakeVelocityLowLeft_b:Vector)
   ) {
    bActorSpaceFakeVel=true
    bShowTrailLocation=true
    bUsePlanarLimit=true
    chainLength=6
    debugLifeTime=0.1
    maxDeltaTime=0.0333
    relaxationSpeedScale=1.2
    alphaScaleBiasClamp {bInterpResult=true}
    baseJoint {boneName=spine_03}
    planarLimits [
     {drivingBone {boneName=root}, planeTransform {location=<0 0 10>, rotation=<rot 0 0 0>, scale=<1 1 1>}}
     {drivingBone {boneName=spine_03}, planeTransform {location=<0 -20 0>, rotation=<rot 0 15 -89.9999>, scale=<1 1 1>}}
    ]
    rotationLimits [
     {limitMax=<180 180 180>}
     {limitMax=<180 180 180>}
     {limitMax=<180 180 180>}
     {limitMax=<180 180 180>}
     {limitMax=<180 180 180>}
     {limitMax=<180 180 180>}
    ]
    rotationOffsets [<0 0 0>, <0 0 0>, <0 0 0>, <0 0 0>, <0 0 0>, <0 0 0>]
    trailBone {boneName=cape_08_l}
    trailRelaxationSpeed {editorCurveData {keys [<at 0 15 RCIM_Linear>,<at 0.5281 8.315 RCIM_Linear>,<at 1 6 RCIM_Linear>]}}
   }
   n30 = Trail controller - Bone: cape_08_r(
    ComponentPose=@n29
    Alpha=1
    RelaxationSpeedScale=@n28
    FakeVelocity=Get CapeTrailFakeVelocityLowRight_b() -> (CapeTrailFakeVelocityLowRight_b:Vector)
   ) {
    bActorSpaceFakeVel=true
    bShowTrailLocation=true
    bUsePlanarLimit=true
    chainLength=6
    debugLifeTime=0.1
    maxDeltaTime=0.0333
    relaxationSpeedScale=1.2
    alphaScaleBiasClamp {bInterpResult=true}
    baseJoint {boneName=spine_03}
    planarLimits [
     {drivingBone {boneName=root}, planeTransform {location=<0 0 10>, rotation=<rot 0 0 0>, scale=<1 1 1>}}
     {drivingBone {boneName=spine_03}, planeTransform {location=<0 -20 0>, rotation=<rot 0 15 -89.9999>, scale=<1 1 1>}}
    ]
    rotationLimits [
     {limitMax=<180 180 180>}
     {limitMax=<180 180 180>}
     {limitMax=<180 180 180>}
     {limitMax=<180 180 180>}
     {limitMax=<180 180 180>}
     {limitMax=<180 180 180>}
    ]
    rotationOffsets [<0 0 0>, <0 0 0>, <0 0 0>, <0 0 0>, <0 0 0>, <0 0 0>]
    trailBone {boneName=cape_08_r}
    trailRelaxationSpeed {editorCurveData {keys [<at 0 15 RCIM_Linear>,<at 0.5281 8.315 RCIM_Linear>,<at 1 6 RCIM_Linear>]}}
   }
   n31 = "Transform (Modify) Bone - Bone: cape_01_l"(ComponentPose=@n30, Alpha=1, Translation="0,-1,0", Rotation="0,0,0", Scale="1,1,1") {translationMode=BMM_Additive, alphaBoolBlend {blendOption=Linear}, boneToModify {boneName=cape_01_l}, translation {y=-1}}
   n32 = Component To Local(ComponentPose=@n31)
   n33 = Get bIsDBNO()
   n34 = "Transform (Modify) Bone - Bone: dyn_crouch_collision"(
    ComponentPose=Local To Component(LocalPose=Use cached pose 'initialPoseCache'())
    Alpha=Get crouch alpha inv() -> (crouch alpha inv:double)
    Translation="0.000,0.000,-49.249"
    Rotation="0,0,0"
    Scale="1,1,1"
   ) {
    translationMode=BMM_Replace
    translationSpace=BCS_BoneSpace
    alphaBoolBlend {blendOption=Linear}
    alphaScaleBiasClamp {bInterpResult=true}
    boneToModify {boneName=dyn_crouch_collision}
    translation {z=-49.249}
   }
   n35 = Get CapeGravityOverride() -> (CapeGravityOverride:Vector)
   n36 = RigidBody(ComponentPose=@n34, OverrideWorldGravity=@n35) {
    bClampLinearTranslationLimitToRefPose=true
    bOverrideWorldGravity=true
    overridePhysicsAsset=&/M_MED_HighTower_Date_Cape_DynCapeKayak
    simulationSpace=BaseBoneSpace
    alphaBoolBlend {blendOption=Linear}
    baseBoneRef {boneName=spine_05}
    componentAppliedLinearAccClamp=<3000 3000 3000>
    componentLinearAccScale=<0.2 0.5 0.5>
    componentLinearVelScale=<0.2 0.5 0.5>
    overrideWorldGravity {z=-980}
   }
   n37 = RigidBody(ComponentPose=@n36, OverrideWorldGravity=@n35) {
    bClampLinearTranslationLimitToRefPose=true
    bOverrideWorldGravity=true
    overridePhysicsAsset=&/M_MED_HighTower_Date_Cape_DynChain
    simulationSpace=BaseBoneSpace
    alphaBoolBlend {blendOption=Linear}
    baseBoneRef {boneName=root}
    componentAppliedLinearAccClamp=<3000 3000 3000>
    componentLinearAccScale=<0.2 0.5 0.5>
    componentLinearVelScale=<0.2 0.5 0.5>
    overrideWorldGravity {z=-980}
   }
   n38 = Control Rig!(
    Source=Blend Poses by bool(
     BlendPose_0=Component To Local(ComponentPose=@n37)
     BlendPose_1=Blend Poses by bool(BlendPose_0=@n18, BlendPose_1=@n32, bActiveValue=@n33) {blendType=Linear}
     bActiveValue=Get bRigidBodyEnabled()
    ) {blendType=Linear}
    Alpha
    isSkydiving:double=Get SkydiveFloat() -> (SkydiveFloat:float)
   ) -> (Pose) {
    alphaBoolBlend {blendOption=Linear}
    controlRigAssetReference {blueprintRigClass=&/M_MED_HighTower_Date_Cape_CtrlRig.M_MED_HighTower_Date_Cape_CtrlRig_C}
   }
   n39 = "Transform (Modify) Bone - Bone: cape_01_r"(ComponentPose=Local To Component(LocalPose=@n38), Alpha=1, Translation="0,-2,0", Rotation="0,0,0", Scale="1,1,1") {
    rotationMode=BMM_Additive
    rotationSpace=BCS_BoneSpace
    scaleMode=BMM_Additive
    scaleSpace=BCS_BoneSpace
    translationMode=BMM_Additive
    translationSpace=BCS_BoneSpace
    alphaBoolBlend {blendOption=Linear}
    boneToModify {boneName=cape_01_r}
    translation {y=-2}
   }
   n40 = "Transform (Modify) Bone - Bone: cape_01_l"(ComponentPose=@n39, Alpha=1, Translation="0,-6,0", Rotation="0,0,0", Scale="1,1,1") {
    rotationMode=BMM_Additive
    rotationSpace=BCS_BoneSpace
    scaleMode=BMM_Additive
    scaleSpace=BCS_BoneSpace
    translationMode=BMM_Additive
    translationSpace=BCS_BoneSpace
    alphaBoolBlend {blendOption=Linear}
    boneToModify {boneName=cape_01_l}
    translation {y=-6}
   }
   n41 = "Transform (Modify) Bone - Bone: cape_01_r"!(ComponentPose=@n40, bAlphaBoolEnabled:bool=@n5, Translation="0,0,0", Rotation="0,-7,0", Scale="1,1,1") -> (Pose) {
    alphaInputType=Bool
    rotationMode=BMM_Additive
    rotationSpace=BCS_BoneSpace
    scaleMode=BMM_Additive
    scaleSpace=BCS_BoneSpace
    translationMode=BMM_Additive
    translationSpace=BCS_BoneSpace
    alphaBoolBlend {blendInTime=0.1, blendOption=Linear, blendOutTime=0.1}
    boneToModify {boneName=cape_01_r}
    rotation {yaw=-7}
   }
   n42 = "Transform (Modify) Bone - Bone: cape_01_l"!(ComponentPose=@n41, bAlphaBoolEnabled:bool=@n5, Translation="0,0,0", Rotation="0,-7,0", Scale="1,1,1") -> (Pose) {
    alphaInputType=Bool
    rotationMode=BMM_Additive
    rotationSpace=BCS_BoneSpace
    scaleMode=BMM_Additive
    scaleSpace=BCS_BoneSpace
    translationMode=BMM_Additive
    translationSpace=BCS_BoneSpace
    alphaBoolBlend {blendInTime=0.1, blendOption=Linear, blendOutTime=0.1}
    boneToModify {boneName=cape_01_l}
    rotation {yaw=-7}
   }
   n43 = "Transform (Modify) Bone - Bone: cape_mid_01_l"!(ComponentPose=@n42, bAlphaBoolEnabled:bool=@n5, Translation="0,0,0", Rotation="0,-7,0", Scale="1,1,1") -> (Pose) {
    alphaInputType=Bool
    rotationMode=BMM_Additive
    rotationSpace=BCS_BoneSpace
    scaleMode=BMM_Additive
    scaleSpace=BCS_BoneSpace
    translationMode=BMM_Additive
    translationSpace=BCS_BoneSpace
    alphaBoolBlend {blendInTime=0.1, blendOption=Linear, blendOutTime=0.1}
    boneToModify {boneName=cape_mid_01_l}
    rotation {yaw=-7}
   }
   n44 = "Transform (Modify) Bone - Bone: cape_mid_01_r"!(ComponentPose=@n43, bAlphaBoolEnabled:bool=@n5, Translation="0,0,0", Rotation="0,-7,0", Scale="1,1,1") -> (Pose) {
    alphaInputType=Bool
    rotationMode=BMM_Additive
    rotationSpace=BCS_BoneSpace
    scaleMode=BMM_Additive
    scaleSpace=BCS_BoneSpace
    translationMode=BMM_Additive
    translationSpace=BCS_BoneSpace
    alphaBoolBlend {blendInTime=0.1, blendOption=Linear, blendOutTime=0.1}
    boneToModify {boneName=cape_mid_01_r}
    rotation {yaw=-7}
   }
   n45 = "Transform (Modify) Bone - Bone: cape_03_l"!(ComponentPose=@n44, bAlphaBoolEnabled:bool=@n4, Translation="0,0,0", Rotation="0,-10,-30", Scale="1,1,1") -> (Pose) {
    alphaInputType=Bool
    rotationMode=BMM_Additive
    rotationSpace=BCS_BoneSpace
    scaleMode=BMM_Additive
    scaleSpace=BCS_BoneSpace
    translationMode=BMM_Additive
    translationSpace=BCS_BoneSpace
    alphaBoolBlend {blendOption=Linear}
    boneToModify {boneName=cape_03_l}
    rotation {roll=-30, yaw=-10}
   }
   n46 = "Transform (Modify) Bone - Bone: cape_06_r"!(ComponentPose=@n45, bAlphaBoolEnabled:bool=@n4, Translation="0,0,0", Rotation="0,-10,0", Scale="1,1,1") -> (Pose) {
    alphaInputType=Bool
    rotationMode=BMM_Additive
    rotationSpace=BCS_BoneSpace
    scaleMode=BMM_Additive
    scaleSpace=BCS_BoneSpace
    translationMode=BMM_Additive
    translationSpace=BCS_BoneSpace
    alphaBoolBlend {blendOption=Linear}
    boneToModify {boneName=cape_06_r}
    rotation {yaw=-10}
   }
   n47 = "Transform (Modify) Bone - Bone: cape_06_l"!(ComponentPose=@n46, bAlphaBoolEnabled:bool=@n4, Translation="0,0,0", Rotation="0,-10,0", Scale="1,1,1") -> (Pose) {
    alphaInputType=Bool
    rotationMode=BMM_Additive
    rotationSpace=BCS_BoneSpace
    scaleMode=BMM_Additive
    scaleSpace=BCS_BoneSpace
    translationMode=BMM_Additive
    translationSpace=BCS_BoneSpace
    alphaBoolBlend {blendOption=Linear}
    boneToModify {boneName=cape_06_l}
    rotation {yaw=-10}
   }
   n48 = "Transform (Modify) Bone - Bone: cape_mid_06_l"!(ComponentPose=@n47, bAlphaBoolEnabled:bool=@n4, Translation="0,0,0", Rotation="0,-10,0", Scale="1,1,1") -> (Pose) {
    alphaInputType=Bool
    rotationMode=BMM_Additive
    rotationSpace=BCS_BoneSpace
    scaleMode=BMM_Additive
    scaleSpace=BCS_BoneSpace
    translationMode=BMM_Additive
    translationSpace=BCS_BoneSpace
    alphaBoolBlend {blendOption=Linear}
    boneToModify {boneName=cape_mid_06_l}
    rotation {yaw=-10}
   }
   n49 = "Transform (Modify) Bone - Bone: cape_mid_06_r"!(ComponentPose=@n48, bAlphaBoolEnabled:bool=@n4, Translation="0,0,0", Rotation="0,-10,0", Scale="1,1,1") -> (Pose) {
    alphaInputType=Bool
    rotationMode=BMM_Additive
    rotationSpace=BCS_BoneSpace
    scaleMode=BMM_Additive
    scaleSpace=BCS_BoneSpace
    translationMode=BMM_Additive
    translationSpace=BCS_BoneSpace
    alphaBoolBlend {blendOption=Linear}
    boneToModify {boneName=cape_mid_06_r}
    rotation {yaw=-10}
   }
   n50 = "Constraint - cape_03_r (P)"(ComponentPose=@n49, Alpha=SelectFloat(A=0.2, B=0, bPickA=@n2), ConstraintWeights_0=1) {
    alphaBoolBlend {blendOption=Linear}
    boneToModify {boneName=cape_03_r}
    constraintSetup [{targetBone {boneName=upperarm_r}}]
    constraintWeights [1]
   }
   n51 = "Transform (Modify) Bone - Bone: cape_01_r"(ComponentPose=@n50, bAlphaBoolEnabled=@n2, Translation="0,-3,0", Rotation="0,-10,0", Scale="1,1,1") {
    alphaInputType=Bool
    rotationMode=BMM_Additive
    rotationSpace=BCS_BoneSpace
    scaleMode=BMM_Additive
    scaleSpace=BCS_BoneSpace
    translationMode=BMM_Additive
    translationSpace=BCS_BoneSpace
    alphaBoolBlend {blendInTime=0.1, blendOption=Linear, blendOutTime=0.1}
    boneToModify {boneName=cape_01_r}
    rotation {yaw=-10}
    translation {y=-3}
   }
   n52 = "Transform (Modify) Bone - Bone: cape_01_l"(ComponentPose=@n51, bAlphaBoolEnabled=@n2, Translation="0,-6,0", Rotation="0,0,0", Scale="1,1,1") {
    alphaInputType=Bool
    rotationMode=BMM_Additive
    rotationSpace=BCS_BoneSpace
    scaleMode=BMM_Additive
    scaleSpace=BCS_BoneSpace
    translationMode=BMM_Additive
    translationSpace=BCS_BoneSpace
    alphaBoolBlend {blendInTime=0.1, blendOption=Linear, blendOutTime=0.1}
    boneToModify {boneName=cape_01_l}
    translation {y=-6}
   }
   n53 = "Transform (Modify) Bone - Bone: cape_01_r"(ComponentPose=@n52, bAlphaBoolEnabled=@n3, Translation="0,0,0", Rotation="0,-9,0", Scale="1,1,1") {
    alphaInputType=Bool
    rotationMode=BMM_Additive
    rotationSpace=BCS_BoneSpace
    scaleMode=BMM_Additive
    scaleSpace=BCS_BoneSpace
    translationMode=BMM_Additive
    translationSpace=BCS_BoneSpace
    alphaBoolBlend {blendInTime=0.1, blendOption=Linear, blendOutTime=0.1}
    boneToModify {boneName=cape_01_r}
    rotation {yaw=-9}
   }
   n54 = "Transform (Modify) Bone - Bone: cape_01_l"(ComponentPose=@n53, bAlphaBoolEnabled=@n3, Translation="0,0,0", Rotation="0,-12,0", Scale="1,1,1") {
    alphaInputType=Bool
    rotationMode=BMM_Additive
    rotationSpace=BCS_BoneSpace
    scaleMode=BMM_Additive
    scaleSpace=BCS_BoneSpace
    translationMode=BMM_Additive
    translationSpace=BCS_BoneSpace
    alphaBoolBlend {blendInTime=0.1, blendOption=Linear, blendOutTime=0.1}
    boneToModify {boneName=cape_01_l}
    rotation {yaw=-12}
   }
   n55 = "Transform (Modify) Bone - Bone: cape_mid_01_l"(ComponentPose=@n54, bAlphaBoolEnabled=@n3, Translation="0,0,0", Rotation="0,-12,0", Scale="1,1,1") {
    alphaInputType=Bool
    rotationMode=BMM_Additive
    rotationSpace=BCS_BoneSpace
    scaleMode=BMM_Additive
    scaleSpace=BCS_BoneSpace
    translationMode=BMM_Additive
    translationSpace=BCS_BoneSpace
    alphaBoolBlend {blendInTime=0.1, blendOption=Linear, blendOutTime=0.1}
    boneToModify {boneName=cape_mid_01_l}
    rotation {yaw=-12}
   }
   n56 = "Transform (Modify) Bone - Bone: cape_mid_01_r"(ComponentPose=@n55, bAlphaBoolEnabled=@n3, Translation="0,0,0", Rotation="0,-9,0", Scale="1,1,1") {
    alphaInputType=Bool
    rotationMode=BMM_Additive
    rotationSpace=BCS_BoneSpace
    scaleMode=BMM_Additive
    scaleSpace=BCS_BoneSpace
    translationMode=BMM_Additive
    translationSpace=BCS_BoneSpace
    alphaBoolBlend {blendInTime=0.1, blendOption=Linear, blendOutTime=0.1}
    boneToModify {boneName=cape_mid_01_r}
    rotation {yaw=-9}
   }
   n57 = "Transform (Modify) Bone - Bone: spine_05"!(ComponentPose=@n56, bAlphaBoolEnabled:bool=True, Translation="0,0,0", Rotation="0,0,0", Scale="0,0,0") -> (Pose) {
    alphaInputType=Bool
    rotationSpace=BCS_BoneSpace
    scaleMode=BMM_Additive
    scaleSpace=BCS_BoneSpace
    translationSpace=BCS_BoneSpace
    alphaBoolBlend {blendInTime=0.1, blendOption=Linear, blendOutTime=0.1}
    boneToModify {boneName=spine_05}
    scale=<0 0 0>
   }
   Output Pose(Result=Component To Local(ComponentPose=@n57)) {layerGroup=None}
   Save cached pose 'initialPoseCache'(Pose=Copy Pose From Mesh() {bUseAttachedParent=true}) {cacheName=initialPoseCache}
  ]}
 EventGraph {nodes [
   Event BlueprintUpdateAnimation()
   > n1 = Sequence()
   > Set Is skydiving Alpha(Is skydiving Alpha:double="To Float (Boolean)"(InBool=Get bIsSkyDiving())) -> (Output_Get:double)
   > n2 = Branch(Condition=Get bIsSkyDiving())
   > Set CapeTrail_Alpha(CapeTrail_Alpha:double=0.5) -> (Output_Get:double)
   Set CapeTrail_Alpha(execute=@n2.else, CapeTrail_Alpha=1)
   n3 = Cast To FortPlayerPawn(execute=@n1.then_1, Object=GetOwningActor())
   n4 = Get ParachuteAttachment!(self:Fort Player Pawn Object Reference=@n3) -> (ParachuteAttachment:Fort Player Parachute Object Reference)
   > Is Valid(InputObject=@n4)
   n5 = Get ParachuteMesh!(self:Fort Player Parachute Object Reference=@n4) -> (ParachuteMesh:Skeletal Mesh Component Budgeted Object Reference)
   > Is Valid(InputObject=@n5)
   n6 = Get SkeletalMesh!(self:Skinned Mesh Component Object Reference=@n5) -> (SkeletalMesh:Skeletal Mesh Object Reference)
   > Is Valid(InputObject=@n6)
   > n7 = Branch(Condition=Get bIsInFrontEnd())
   > Set IsDateGlider(IsDateGlider:bool="Equal (SoftObjectReference)"(
     A:"Soft Object Reference (by ref)"=Make Soft Reference!(Input:Skeletal Mesh Object Reference=@n6) -> (Output:Skeletal Mesh Soft Object Reference)
     B:"Soft Object Reference (by ref)"=Get DateGlider() -> (DateGlider:Skeletal Mesh Soft Object Reference)
    ) -> (ReturnValue:bool=~) {is=title K2Node_PromotableOperator}) -> (Output_Get:bool)
   Set IsDateGlider(execute=@n7.else, IsDateGlider=false)
   Event BlueprintInitializeAnimation()
   n8 = TryGetPawnOwner()
   > Is Valid(InputObject=@n8)
   > n9 = Cast To FortPlayerPawnAthena(Object=@n8)
   > Set player_pawn(player_pawn:Fort Player Pawn Athena Object Reference=@n9) -> (Output_Get:Fort Player Pawn Athena Object Reference)
   Event OnEndCrouching()
   > Set crouch alpha(crouch alpha:double=0) -> (Output_Get:double)
   > Set crouch alpha inv(crouch alpha inv:double=1) -> (Output_Get:double)
   Event OnBeginCrouching()
   > Set crouch alpha(crouch alpha=1)
   > Set crouch alpha inv(crouch alpha inv=0)
   n10 = Event OnEndSkydiving()
   Delay(execute=[@n10.then, @n11.then], Duration=0.1)
   > ResetDynamics(InTeleportType=TeleportPhysics)
   > ResetDynamics(InTeleportType=ResetPhysics)
   n11 = Event OnBeginSkydiving()
   TryGetPawnOwner()
  ]}
]
variables [
 CapeTrail_Alpha {type=number}
 CapeTrailFakeVelocityLowLeft_b {type=Vector}
 CapeTrailFakeVelocityLowRight_b {type=Vector}
 CapeTrailFakeVelocityMidLeft_b {type=Vector}
 CapeTrailFakeVelocityMidRight_b {type=Vector}
 crouch alpha {type=number}
 crouch alpha inv {type=number}
 DateGlider {type=SkeletalMesh, default=/BRCosmetics/Gadgets/Assets/VinderTech_GliderChute/Glider_HighTower_Date/Meshes/Glider_HighTower_Date}
 HeadMesh {type=SkeletalMeshComponent}
 Is skydiving Alpha {type=number}
 IsDateGlider {type=boolean}
 player_pawn {type=FortPlayerPawnAthena}
]
